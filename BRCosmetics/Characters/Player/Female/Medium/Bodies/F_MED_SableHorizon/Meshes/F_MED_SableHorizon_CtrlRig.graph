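# Control Rig for the F_MED_SableHorizon player body that procedurally adjusts skirt and belt bones to follow the character's pose.
# Spherical pose readers, calf bend, thigh rotation and animation attributes weight additive bone offsets, with crouch, slide, emote and cape variants.
# Weighted parent constraints tie belts to body bones and skirt segments to legs.
# Head and neck swing, measured against construction-time reference controls, yields collar alphas.
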ControlRigBlueprint /BRCosmetics/Characters/Player/Female/Medium/Bodies/F_MED_SableHorizon/Meshes/F_MED_SableHorizon_CtrlRig
graphs [
 Add_1_ContainedGraph {nodes [
   n1 = RigVMFunctionEntryNode()
   Add = &RigVMFunction_MathDoubleAdd(A=@n1.A, B=@n1.B)
   RigVMFunctionReturnNode(Result=&RigVMFunction_MathDoubleAdd(A=@Add.Result, B=@n1.C).Result)
  ]}
 Maximum_2_ContainedGraph {nodes [
   n1 = RigVMFunctionEntryNode()
   RigVMFunctionReturnNode(Result=&RigVMFunction_MathDoubleMax(B=@n1.C, A=&RigVMFunction_MathDoubleMax(A=@n1.A, B=@n1.B).Result).Result)
  ]}
 Maximum_2_ContainedGraph {nodes [
   n1 = RigVMFunctionEntryNode()
   RigVMFunctionReturnNode(Result=&RigVMFunction_MathDoubleMax(B=@n1.C, A=&RigVMFunction_MathDoubleMax(A=@n1.A, B=@n1.B).Result).Result)
  ]}
 New Function_ContainedGraph {nodes [
   RigVMFunctionEntryNode()
   > RigVMFunction_Sequence = RigVMAggregateNode()
   n1 = &RigVMFunction_ControlFlowBranch(ExecuteContext=@RigVMFunction_Sequence.A, Condition=RigVMVariableNode() {variable=isSlidingVar_2}.Value)
   RigVMFunction_MathVectorDistance = &RigVMFunction_MathVectorDistance(
    A=&RigUnit_GetTransform(Item=(Type=Bone,Name="dyn_skirt_b_l_01"), bInitial=False).Transform.Translation
    B=&RigUnit_GetTransform(Item=(Type=Bone,Name="foot_r"), bInitial=False).Transform.Translation
   )
   Remap = &RigVMFunction_MathDoubleRemap(SourceMaximum=48, SourceMinimum=50, TargetMaximum=1, TargetMinimum=0, Value=@RigVMFunction_MathVectorDistance.Result)
   &RigUnit_ModifyTransforms(
    ItemToModify="((Item=(Type=Bone,Name=\"dyn_skirt_b_l_base\"),Transform=(Rotation=(X=-0.1274,Y=-0.0283,Z=0.2146,W=0.9679),Translation=(X=-20,Y=20,Z=0),Scale3D=(X=1,Y=1,Z=1))),(Item=(Type=Bone,Name=\"dyn_skirt_b_r_base\"),Transform=(Rotation=(X=0,Y=0,Z=0,W=1),Translation=(X=- ...(truncated)"
    WeightMinimum=0
    Weight=@Remap.Result
    ExecutePin=@n1.True
   )
   n2 = &RigVMFunction_ControlFlowBranch(ExecuteContext=@RigVMFunction_Sequence.B, Condition=RigVMVariableNode() {variable=isSlidingVar_1}.Value)
   &RigUnit_ModifyTransforms(
    ItemToModify="((Item=(Type=Bone,Name=\"dyn_skirt_b_l_base\"),Transform=(Rotation=(X=0,Y=0,Z=0,W=1),Translation=(X=0,Y=0,Z=0),Scale3D=(X=0.25,Y=1,Z=1))),(Item=(Type=Bone,Name=\"dyn_skirt_b_r_base\"),Transform=(Rotation=(X=0,Y=0,Z=0,W=1),Translation=(X=0,Y=0,Z=0),Scale3D=(X=0 ...(truncated)"
    WeightMinimum=0
    Weight=@Remap.Result
    ExecutePin=@n2.True
   )
   RigVMFunctionReturnNode(ExecuteContext=@RigVMFunction_Sequence.C)
  ]}
 New Function_ContainedGraph {nodes [
   RigVMFunctionEntryNode()
   > DeltaRotY = RigVMFunctionReferenceNode(Child=thigh_l, Parent=Ref_Thigh_L, Source Maximum=-70, Source Minimum=-5)
   > RigVMVariableNode(Value=@DeltaRotY.Delta Rotation) {variable=Thigh_L_Side}
   > DeltaRotY_1 = RigVMFunctionReferenceNode(Child=thigh_r, Parent=Ref_Thigh_R, Source Maximum=-70, Source Minimum=-5)
   > RigVMVariableNode(Value=@DeltaRotY_1.Delta Rotation) {variable=Thigh_R_Side}
   > &RigUnit_ParentConstraint(
    AdvancedSettings=(RotationOrderForFilter=XZY)
    Child=(Type=Bone,Name="dyn_beltfront_r")
    Parents="((Item=(Type=Bone,Name=\"spine_01\"),Weight=0.0013),(Item=(Type=Bone,Name=\"thigh_twist_01_l\"),Weight=0.0033),(Item=(Type=Bone,Name=\"deform_glute_r\"),Weight=0.0122),(Item=(Type=Bone,Name=\"deform_glute_l\"),Weight=0.0148),(Item=(Type=Bone,Name=\"spine_02\"),Weigh ...(truncated)"
    Weight=1
   )
   > &RigUnit_ParentConstraint(
    AdvancedSettings=(RotationOrderForFilter=XZY)
    Child=(Type=Bone,Name="dyn_beltfront_l")
    Parents="((Item=(Type=Bone,Name=\"spine_01\"),Weight=0.0011),(Item=(Type=Bone,Name=\"spine_03\"),Weight=0.0012),(Item=(Type=Bone,Name=\"deform_glute_l\"),Weight=0.0031),(Item=(Type=Bone,Name=\"deform_glute_r\"),Weight=0.0038),(Item=(Type=Bone,Name=\"thigh_twist_01_l\"),Weigh ...(truncated)"
    Weight=1
   )
   > &RigUnit_ParentConstraint(
    AdvancedSettings=(RotationOrderForFilter=XZY)
    Child=(Type=Bone,Name="dyn_beltfront")
    Parents="((Item=(Type=Bone,Name=\"thigh_r\"),Weight=0.0013),(Item=(Type=Bone,Name=\"thigh_l\"),Weight=0.002),(Item=(Type=Bone,Name=\"spine_02\"),Weight=0.029),(Item=(Type=Bone,Name=\"deform_groin_l\"),Weight=0.0565),(Item=(Type=Bone,Name=\"deform_groin_r\"),Weight=0.0586),(I ...(truncated)"
    Weight=1
   )
   > &RigUnit_ParentConstraint(
    AdvancedSettings=(RotationOrderForFilter=XZY)
    Child=(Type=Bone,Name="dyn_belt_back")
    Parents="((Item=(Type=Bone,Name=\"thigh_l\"),Weight=0.0012),(Item=(Type=Bone,Name=\"spine_02\"),Weight=0.0018),(Item=(Type=Bone,Name=\"thigh_twist_01_r\"),Weight=0.0332),(Item=(Type=Bone,Name=\"thigh_twist_01_l\"),Weight=0.0393),(Item=(Type=Bone,Name=\"deform_glute_r\"),Weig ...(truncated)"
    Weight=1
   )
   > RigVMFunctionReturnNode()
  ]}
 New Function_ContainedGraph {nodes [
   n1 = RigVMFunctionEntryNode()
   RigVMFunction_MathQuaternionToEuler = &RigVMFunction_MathQuaternionToEuler(
    RotationOrder=XYZ
    Value=&RigUnit_GetRelativeTransformForItem(
     Child=(Type=Bone,Name="hand_l")
     Parent=(Type=Null,Name="lowerarm_l")
     bChildInitial=False
     bParentInitial=False
     Child.Name=@n1.Child
     Parent.Name=@n1.Parent
    ).RelativeTransform.Rotation
   )
   Remap = &RigVMFunction_MathDoubleRemap(
    TargetMaximum=1
    TargetMinimum=0
    SourceMaximum=@n1.Source Maximum
    SourceMinimum=@n1.Source Minimum
    Value=@RigVMFunction_MathQuaternionToEuler.Result.Y
   )
   > RigVMFunctionReturnNode(Delta Rotation=@Remap.Result)
  ]}
 New Function_ContainedGraph {nodes [
   RigVMFunctionEntryNode()
   > RigVMFunction_Sequence = RigVMAggregateNode()
   n1 = &RigUnit_SphericalPoseReader(
    ActiveRegionScaleFactors=(PositiveWidth=4,NegativeWidth=3,PositiveHeight=2.5,NegativeHeight=2)
    ActiveRegionSize=0.1
    Debug=(bDrawDebug=false,bDraw2D=False,bDrawLocalAxes=False,DebugScale=25,DebugSegments=20,DebugThickness=0.25)
    DriverAxis=(X=-1,Y=0,Z=0)
    DriverItem=(Type=Bone,Name="thigh_l")
    FalloffRegionScaleFactors=(PositiveWidth=1,NegativeWidth=3,PositiveHeight=1,NegativeHeight=0.5)
    FalloffSize=0.35
    FlipHeightScaling=False
    FlipWidthScaling=False
    OptionalParentItem=(Type=Bone,Name="None")
    RotationOffset=(X=100,Y=0,Z=-20)
    ExecutePin=@RigVMFunction_Sequence.A
   )
   > n2 = &RigUnit_SphericalPoseReader(
    ActiveRegionScaleFactors=(PositiveWidth=4,NegativeWidth=3,PositiveHeight=2.5,NegativeHeight=2)
    ActiveRegionSize=0.1
    Debug=(bDrawDebug=false,bDraw2D=False,bDrawLocalAxes=False,DebugScale=25,DebugSegments=20,DebugThickness=0.25)
    DriverAxis=(X=1,Y=0,Z=0)
    DriverItem=(Type=Bone,Name="thigh_r")
    FalloffRegionScaleFactors=(PositiveWidth=1,NegativeWidth=3.6,PositiveHeight=1,NegativeHeight=0.5)
    FalloffSize=0.4
    FlipHeightScaling=False
    FlipWidthScaling=true
    OptionalParentItem=(Type=Bone,Name="None")
    RotationOffset=(X=-88.1205,Y=0,Z=-14.0228)
   )
   > n3 = &RigUnit_SphericalPoseReader(
    ActiveRegionScaleFactors=(PositiveWidth=3,NegativeWidth=2,PositiveHeight=2,NegativeHeight=2)
    ActiveRegionSize=0.1
    Debug=(bDrawDebug=false,bDraw2D=False,bDrawLocalAxes=False,DebugScale=25,DebugSegments=20,DebugThickness=0.25)
    DriverAxis=(X=-1,Y=0,Z=0)
    DriverItem=(Type=Bone,Name="thigh_l")
    FalloffRegionScaleFactors=(PositiveWidth=2,NegativeWidth=2,PositiveHeight=2,NegativeHeight=2)
    FalloffSize=0.25
    FlipHeightScaling=false
    FlipWidthScaling=false
    OptionalParentItem=(Type=Bone,Name="None")
    RotationOffset=(X=180,Y=-40,Z=0)
   )
   > n4 = &RigUnit_SphericalPoseReader(
    ActiveRegionScaleFactors=(PositiveWidth=3,NegativeWidth=2,PositiveHeight=2,NegativeHeight=2)
    ActiveRegionSize=0.1
    Debug=(bDrawDebug=false,bDraw2D=False,bDrawLocalAxes=False,DebugScale=25,DebugSegments=20,DebugThickness=0.25)
    DriverAxis=(X=1,Y=0,Z=0)
    DriverItem=(Type=Bone,Name="thigh_r")
    FalloffRegionScaleFactors=(PositiveWidth=2,NegativeWidth=2,PositiveHeight=2,NegativeHeight=2)
    FalloffSize=0.25
    FlipHeightScaling=False
    FlipWidthScaling=true
    OptionalParentItem=(Type=Bone,Name="None")
    RotationOffset=(X=0,Y=-40,Z=0)
   )
   > n5 = &RigUnit_SphericalPoseReader(
    ActiveRegionScaleFactors=(PositiveWidth=2,NegativeWidth=2,PositiveHeight=2,NegativeHeight=1)
    ActiveRegionSize=0.1
    Debug=(bDrawDebug=false,bDraw2D=false,bDrawLocalAxes=False,DebugScale=25,DebugSegments=20,DebugThickness=0.25)
    DriverAxis=(X=-1,Y=0,Z=0)
    DriverItem=(Type=Bone,Name="thigh_l")
    FalloffRegionScaleFactors=(PositiveWidth=2,NegativeWidth=2,PositiveHeight=4,NegativeHeight=1)
    FalloffSize=0.27
    FlipHeightScaling=false
    FlipWidthScaling=false
    OptionalParentItem=(Type=Bone,Name="None")
    RotationOffset=(X=225,Y=90,Z=-45)
   )
   > n6 = &RigUnit_SphericalPoseReader(
    ActiveRegionScaleFactors=(PositiveWidth=2,NegativeWidth=2,PositiveHeight=2,NegativeHeight=1)
    ActiveRegionSize=0.1
    Debug=(bDrawDebug=false,bDraw2D=False,bDrawLocalAxes=False,DebugScale=25,DebugSegments=20,DebugThickness=0.25)
    DriverAxis=(X=1,Y=0,Z=0)
    DriverItem=(Type=Bone,Name="thigh_r")
    FalloffRegionScaleFactors=(PositiveWidth=2,NegativeWidth=2,PositiveHeight=3,NegativeHeight=1)
    FalloffSize=0.27
    FlipHeightScaling=false
    FlipWidthScaling=false
    OptionalParentItem=(Type=Bone,Name="None")
    RotationOffset=(X=160,Y=84.6752,Z=60.6503)
   )
   > n7 = &RigUnit_SphericalPoseReader(
    ActiveRegionScaleFactors=(PositiveWidth=1.5,NegativeWidth=1.5,PositiveHeight=1.5,NegativeHeight=1.5)
    ActiveRegionSize=0.1
    Debug=(bDrawDebug=false,bDraw2D=False,bDrawLocalAxes=False,DebugScale=25,DebugSegments=20,DebugThickness=0.25)
    DriverAxis=(X=-1,Y=0,Z=0)
    DriverItem=(Type=Bone,Name="thigh_l")
    FalloffRegionScaleFactors=(PositiveWidth=2,NegativeWidth=2,PositiveHeight=2,NegativeHeight=2)
    FalloffSize=0.23
    FlipHeightScaling=false
    FlipWidthScaling=false
    OptionalParentItem=(Type=Bone,Name="None")
    RotationOffset=(X=33,Y=35,Z=-2.2671)
   )
   > n8 = &RigUnit_SphericalPoseReader(
    ActiveRegionScaleFactors=(PositiveWidth=1,NegativeWidth=1,PositiveHeight=1,NegativeHeight=1)
    ActiveRegionSize=0.12
    Debug=(bDrawDebug=true,bDraw2D=False,bDrawLocalAxes=False,DebugScale=25,DebugSegments=20,DebugThickness=0.25)
    DriverAxis=(X=1,Y=0,Z=0)
    DriverItem=(Type=Bone,Name="thigh_r")
    FalloffRegionScaleFactors=(PositiveWidth=1,NegativeWidth=1,PositiveHeight=1,NegativeHeight=1)
    FalloffSize=0.25
    FlipHeightScaling=False
    FlipWidthScaling=true
    OptionalParentItem=(Type=Bone,Name="None")
    RotationOffset=(X=-180,Y=50,Z=25)
   )
   > n9 = &RigUnit_SphericalPoseReader(
    ActiveRegionScaleFactors=(PositiveWidth=1,NegativeWidth=1,PositiveHeight=1,NegativeHeight=1)
    ActiveRegionSize=0.12
    Debug=(bDrawDebug=false,bDraw2D=False,bDrawLocalAxes=False,DebugScale=25,DebugSegments=20,DebugThickness=0.25)
    DriverAxis=(X=0,Y=1,Z=0)
    DriverItem=(Type=Bone,Name="thigh_l")
    FalloffRegionScaleFactors=(PositiveWidth=1,NegativeWidth=1,PositiveHeight=1,NegativeHeight=1)
    FalloffSize=0.17
    FlipHeightScaling=false
    FlipWidthScaling=false
    OptionalParentItem=(Type=Bone,Name="None")
    RotationOffset=(X=58.5993,Y=-9.6785,Z=-26.2186)
   )
   > n10 = &RigUnit_SphericalPoseReader(
    ActiveRegionScaleFactors=(PositiveWidth=1,NegativeWidth=1,PositiveHeight=1,NegativeHeight=1)
    ActiveRegionSize=0.12
    Debug=(bDrawDebug=false,bDraw2D=False,bDrawLocalAxes=False,DebugScale=25,DebugSegments=20,DebugThickness=0.25)
    DriverAxis=(X=0,Y=-1,Z=0)
    DriverItem=(Type=Bone,Name="thigh_r")
    FalloffRegionScaleFactors=(PositiveWidth=1,NegativeWidth=1,PositiveHeight=1,NegativeHeight=1)
    FalloffSize=0.17
    FlipHeightScaling=False
    FlipWidthScaling=true
    OptionalParentItem=(Type=Bone,Name="None")
    RotationOffset=(X=-127.0157,Y=4.3906,Z=-18.4663)
   )
   > n11 = &RigUnit_SphericalPoseReader(
    ActiveRegionScaleFactors=(PositiveWidth=1,NegativeWidth=1,PositiveHeight=1,NegativeHeight=1)
    ActiveRegionSize=0.12
    Debug=(bDrawDebug=false,bDraw2D=False,bDrawLocalAxes=False,DebugScale=25,DebugSegments=20,DebugThickness=0.25)
    DriverAxis=(X=-1,Y=0,Z=0)
    DriverItem=(Type=Bone,Name="thigh_l")
    FalloffRegionScaleFactors=(PositiveWidth=1,NegativeWidth=1,PositiveHeight=1,NegativeHeight=1)
    FalloffSize=0.25
    FlipHeightScaling=false
    FlipWidthScaling=false
    OptionalParentItem=(Type=Bone,Name="None")
    RotationOffset=(X=75.9774,Y=-9.6785,Z=-2.3149)
   )
   > n12 = &RigUnit_SphericalPoseReader(
    ActiveRegionScaleFactors=(PositiveWidth=1,NegativeWidth=1,PositiveHeight=1,NegativeHeight=1)
    ActiveRegionSize=0.12
    Debug=(bDrawDebug=false,bDraw2D=False,bDrawLocalAxes=False,DebugScale=25,DebugSegments=20,DebugThickness=0.25)
    DriverAxis=(X=1,Y=0,Z=0)
    DriverItem=(Type=Bone,Name="thigh_r")
    FalloffRegionScaleFactors=(PositiveWidth=1,NegativeWidth=1,PositiveHeight=1,NegativeHeight=1)
    FalloffSize=0.25
    FlipHeightScaling=False
    FlipWidthScaling=true
    OptionalParentItem=(Type=Bone,Name="None")
    RotationOffset=(X=-106.852,Y=57.2151,Z=33.6105)
   )
   > n13 = &RigUnit_SphericalPoseReader(
    ActiveRegionScaleFactors=(PositiveWidth=1,NegativeWidth=1,PositiveHeight=1,NegativeHeight=1)
    ActiveRegionSize=0.12
    Debug=(bDrawDebug=false,bDraw2D=False,bDrawLocalAxes=False,DebugScale=25,DebugSegments=20,DebugThickness=0.25)
    DriverAxis=(X=-1,Y=0,Z=0)
    DriverItem=(Type=Bone,Name="thigh_l")
    FalloffRegionScaleFactors=(PositiveWidth=1,NegativeWidth=1,PositiveHeight=1,NegativeHeight=1)
    FalloffSize=0.3
    FlipHeightScaling=false
    FlipWidthScaling=false
    OptionalParentItem=(Type=Bone,Name="None")
    RotationOffset=(X=45.4793,Y=-47.9869,Z=88.1318)
   )
   > n14 = &RigUnit_SphericalPoseReader(
    ActiveRegionScaleFactors=(PositiveWidth=1,NegativeWidth=1,PositiveHeight=1,NegativeHeight=1)
    ActiveRegionSize=0.12
    Debug=(bDrawDebug=false,bDraw2D=False,bDrawLocalAxes=False,DebugScale=25,DebugSegments=20,DebugThickness=0.25)
    DriverAxis=(X=1,Y=0,Z=0)
    DriverItem=(Type=Bone,Name="thigh_r")
    FalloffRegionScaleFactors=(PositiveWidth=1,NegativeWidth=1,PositiveHeight=1,NegativeHeight=1)
    FalloffSize=0.3
    FlipHeightScaling=False
    FlipWidthScaling=true
    OptionalParentItem=(Type=Bone,Name="None")
    RotationOffset=(X=-110.9682,Y=15.2115,Z=-14.8802)
   )
   > n15 = &RigUnit_SphericalPoseReader(
    ActiveRegionScaleFactors=(PositiveWidth=1,NegativeWidth=1.2,PositiveHeight=1,NegativeHeight=1)
    ActiveRegionSize=0.13
    Debug=(bDrawDebug=false,bDraw2D=False,bDrawLocalAxes=False,DebugScale=25,DebugSegments=20,DebugThickness=0.25)
    DriverAxis=(X=-1,Y=0,Z=0)
    DriverItem=(Type=Bone,Name="thigh_l")
    FalloffRegionScaleFactors=(PositiveWidth=1,NegativeWidth=1.2,PositiveHeight=1,NegativeHeight=1)
    FalloffSize=0.25
    FlipHeightScaling=false
    FlipWidthScaling=false
    OptionalParentItem=(Type=Bone,Name="None")
    RotationOffset=(X=135,Y=-150,Z=90)
   )
   > n16 = &RigUnit_SphericalPoseReader(
    ActiveRegionScaleFactors=(PositiveWidth=1,NegativeWidth=1,PositiveHeight=1,NegativeHeight=1)
    ActiveRegionSize=0.12
    Debug=(bDrawDebug=true,bDraw2D=False,bDrawLocalAxes=False,DebugScale=25,DebugSegments=20,DebugThickness=0.25)
    DriverAxis=(X=1,Y=0,Z=0)
    DriverItem=(Type=Bone,Name="thigh_r")
    FalloffRegionScaleFactors=(PositiveWidth=1,NegativeWidth=1,PositiveHeight=1,NegativeHeight=1)
    FalloffSize=0.25
    FlipHeightScaling=False
    FlipWidthScaling=true
    OptionalParentItem=(Type=Bone,Name="None")
    RotationOffset=(X=-125,Y=45,Z=25)
   )
   > n17 = &RigUnit_SphericalPoseReader(
    ActiveRegionScaleFactors=(PositiveWidth=1,NegativeWidth=1,PositiveHeight=1,NegativeHeight=1)
    ActiveRegionSize=0.12
    Debug=(bDrawDebug=true,bDraw2D=False,bDrawLocalAxes=False,DebugScale=25,DebugSegments=20,DebugThickness=0.25)
    DriverAxis=(X=0,Y=1,Z=0)
    DriverItem=(Type=Bone,Name="thigh_l")
    FalloffRegionScaleFactors=(PositiveWidth=1,NegativeWidth=1,PositiveHeight=1,NegativeHeight=1)
    FalloffSize=0.17
    FlipHeightScaling=false
    FlipWidthScaling=false
    OptionalParentItem=(Type=Bone,Name="None")
    RotationOffset=(X=147.4193,Y=-9.6785,Z=-26.2186)
   )
   > n18 = &RigUnit_SphericalPoseReader(
    ActiveRegionScaleFactors=(PositiveWidth=1,NegativeWidth=1,PositiveHeight=1,NegativeHeight=1)
    ActiveRegionSize=0.12
    Debug=(bDrawDebug=false,bDraw2D=False,bDrawLocalAxes=False,DebugScale=25,DebugSegments=20,DebugThickness=0.25)
    DriverAxis=(X=1,Y=0,Z=0)
    DriverItem=(Type=Bone,Name="calf_r")
    FalloffRegionScaleFactors=(PositiveWidth=1,NegativeWidth=1,PositiveHeight=1,NegativeHeight=1)
    FalloffSize=0.3
    FlipHeightScaling=false
    FlipWidthScaling=false
    OptionalParentItem=(Type=Bone,Name="None")
    RotationOffset=(X=0,Y=90,Z=-45)
   )
   > RigVMVariableNode(Value=@n18.OutputParam) {variable=crouchFloa2}
   > &RigUnit_ModifyTransforms(
    ItemToModify=((Item=(Type=Bone,Name="belt_front_main"),Transform=(Rotation=(X=0,Y=0,Z=0,W=1),Translation=(X=0,Y=0,Z=1),Scale3D=(X=1,Y=1,Z=1))))
    Weight=1
    WeightMinimum=0
   )
   > &RigUnit_ModifyTransforms(
    ItemToModify=((Item=(Type=Bone,Name="base_belt_back"),Transform=(Rotation=(X=0,Y=-0.5,Z=0,W=0.866),Translation=(X=0,Y=0,Z=5),Scale3D=(X=1,Y=1,Z=1))))
    Weight=1
    WeightMinimum=0
   )
   Multiply = &RigVMFunction_MathDoubleMul(A=RigVMVariableNode() {variable=R thighBack Alpha}.Value, B=RigVMVariableNode() {variable=L thighForward Alpha}.Value)
   > RigVMVariableNode(Value=@Multiply.Result) {variable=ThighSplit L Fwd}
   > RigVMVariableNode(Value=&RigVMFunction_MathDoubleMul(A=RigVMVariableNode() {variable=R thighForward Alpha}.Value, B=RigVMVariableNode() {variable=L thighBack Alpha}.Value).Result) {variable=ThighSplit R Fwd}
   RigVMFunction_MathQuaternionToEuler = &RigVMFunction_MathQuaternionToEuler(
    RotationOrder=ZYX
    Value=&RigUnit_GetRelativeTransformForItem(Child=(Type=Bone,Name="calf_l"), Parent=(Type=Bone,Name="thigh_l"), bChildInitial=False, bParentInitial=False).RelativeTransform.Rotation
   )
   Remap = &RigVMFunction_MathDoubleRemap(SourceMaximum=90, SourceMinimum=14, TargetMaximum=1, TargetMinimum=0, Value=@RigVMFunction_MathQuaternionToEuler.Result.Z)
   > RigVMVariableNode(Value=@Remap.Result) {variable=L calfBend Alpha}
   n19 = &RigUnit_GetRelativeTransformForItem(Child=(Type=Bone,Name="calf_r"), Parent=(Type=Bone,Name="thigh_r"), bChildInitial=False, bParentInitial=False)
   > RigVMVariableNode(Value=&RigVMFunction_MathDoubleRemap(
     SourceMaximum=90
     SourceMinimum=14
     TargetMaximum=1
     TargetMinimum=0
     Value=&RigVMFunction_MathQuaternionToEuler(RotationOrder=ZYX, Value=@n19.RelativeTransform.Rotation).Result.Z
    ).Result) {variable=R calfBend Alpha}
   > &RigUnit_ModifyTransforms(
    ItemToModify="((Item=(Type=Bone,Name=\"jnt_support_leg_l\"),Transform=(Rotation=(X=0,Y=0,Z=0,W=1),Translation=(X=0,Y=0,Z=0),Scale3D=(X=0.8,Y=0.8,Z=0.8))),(Item=(Type=Bone,Name=\"jnt_support_leg_r\"),Transform=(Rotation=(X=0,Y=0,Z=0,W=1),Translation=(X=0,Y=0,Z=0),Scale3D=(X= ...(truncated)"
    Weight=1
    WeightMinimum=0
   )
   RigVMVariableNode(ExecuteContext=@RigVMFunction_Sequence.B, Value=@n1.OutputParam) {variable=L thighForward Alpha}
   > RigVMVariableNode(Value=@n2.OutputParam) {variable=R thighForward Alpha}
   > RigVMVariableNode(Value=@n3.OutputParam) {variable=L thighSide Alpha}
   > RigVMVariableNode(Value=@n4.OutputParam) {variable=R thighSide Alpha}
   > RigVMVariableNode(Value=@n5.OutputParam) {variable=L thighBack Alpha}
   > RigVMVariableNode(Value=@n6.OutputParam) {variable=R thighBack Alpha}
   > RigVMVariableNode(Value=@n7.OutputParam) {variable=L thighSideIn_Fwd Alpha}
   > RigVMVariableNode(Value=@n8.OutputParam) {variable=R thighSideIn_Bwd Alpha}
   > RigVMVariableNode(Value=@n9.OutputParam) {variable=L thighTwistIn Alpha}
   > RigVMVariableNode(Value=@n10.OutputParam) {variable=R thighTwistIn Alpha}
   > RigVMVariableNode(Value=@n11.OutputParam) {variable=L Crouch Strafe Front}
   > RigVMVariableNode(Value=@n12.OutputParam) {variable=R Crouch Strafe Behind}
   > RigVMVariableNode(Value=@n13.OutputParam) {variable=L Crouch Strafe Behind}
   > RigVMVariableNode(Value=@n14.OutputParam) {variable=R Crouch Strafe Front}
   > RigVMVariableNode(Value=@n15.OutputParam) {variable=L thighSideIn_Bwd Alpha}
   > RigVMVariableNode(Value=@n16.OutputParam) {variable=R thighSideIn_Fwd Alpha}
   > RigVMVariableNode(Value=@n17.OutputParam) {variable=L thighTwistOut Alpha}
   RigVMFunctionReturnNode(ExecuteContext=@RigVMFunction_Sequence.I)
   &RigUnit_ModifyTransforms(
    ItemToModify="((Item=(Type=Bone,Name=\"dyn_skirt_f_c_base\"),Transform=(Rotation=(X=0.1913,Y=-0.4619,Z=0.3314,W=0.8001),Translation=(X=0,Y=0,Z=0),Scale3D=(X=1,Y=1,Z=1))),(Item=(Type=Bone,Name=\"dyn_skirt_f_l_base\"),Transform=(Rotation=(X=0.0094,Y=-0.2162,Z=0.0426,W=0.9754) ...(truncated)"
    WeightMinimum=0
    Weight=RigVMDispatchNode(BoneName=root, Default=0, Name=deformRig_l_thigh_up_alpha) {template="DISPATCH_RigDispatch_GetAnimAttribute(in Name,in BoneName,in Default,out Value,out Found)"}.Value
    ExecutePin=@RigVMFunction_Sequence.C
   )
   > n20 = &RigUnit_ModifyTransforms(
    ItemToModify="((Item=(Type=Bone,Name=\"dyn_skirt_f_c_base\"),Transform=(Rotation=(X=-0.1913,Y=-0.4619,Z=-0.3314,W=0.8001),Translation=(X=0,Y=0,Z=0),Scale3D=(X=1,Y=1,Z=1))),(Item=(Type=Bone,Name=\"dyn_skirt_f_r_base\"),Transform=(Rotation=(X=-0.0094,Y=-0.2162,Z=-0.0426,W=0.9 ...(truncated)"
    WeightMinimum=0
    Weight=RigVMDispatchNode(BoneName=root, Default=0, Name=deformRig_r_thigh_up_alpha) {template="DISPATCH_RigDispatch_GetAnimAttribute(in Name,in BoneName,in Default,out Value,out Found)"}.Value
   )
   n21 = &RigVMFunction_ControlFlowBranch(ExecuteContext=@n20.ExecutePin, Condition=RigVMVariableNode() {variable=Is Crouching_0}.Value)
   &RigUnit_ModifyTransforms(
    ItemToModify="((Item=(Type=Bone,Name=\"dyn_skirt_f_l_base\"),Transform=(Rotation=(X=0,Y=-0.0872,Z=0,W=0.9962),Translation=(X=0,Y=0,Z=0),Scale3D=(X=1,Y=1,Z=1))),(Item=(Type=Bone,Name=\"dyn_skirt_s_l_base\"),Transform=(Rotation=(X=0,Y=-0.0872,Z=0,W=0.9962),Translation=(X=-7,Y ...(truncated)"
    WeightMinimum=0
    ExecutePin=@n21.False
    Weight=RigVMVariableNode() {variable=Thigh_L_Side}.Value
   )
   > &RigUnit_ModifyTransforms(
    ItemToModify="((Item=(Type=Bone,Name=\"dyn_skirt_f_r_base\"),Transform=(Rotation=(X=0,Y=-0.0872,Z=0,W=0.9962),Translation=(X=0,Y=0,Z=0),Scale3D=(X=1,Y=1,Z=1))),(Item=(Type=Bone,Name=\"dyn_skirt_s_r_base\"),Transform=(Rotation=(X=0,Y=-0.0872,Z=0,W=0.9962),Translation=(X=0,Y= ...(truncated)"
    WeightMinimum=0
    Weight=RigVMVariableNode() {variable=Thigh_R_Side}.Value
   )
   > &RigUnit_ModifyTransforms(
    ItemToModify="((Item=(Type=Bone,Name=\"dyn_skirt_b_l_base\"),Transform=(Rotation=(X=0,Y=-0.2588,Z=0,W=0.9659),Translation=(X=0,Y=0,Z=5),Scale3D=(X=1,Y=1,Z=1))),(Item=(Type=Bone,Name=\"dyn_skirt_s_r_base\"),Transform=(Rotation=(X=0,Y=-0.0872,Z=0,W=0.9962),Translation=(X=0,Y= ...(truncated)"
    WeightMinimum=0
    Weight=RigVMVariableNode() {variable=R thighBack Alpha}.Value
   )
   > &RigUnit_ModifyTransforms(
    ItemToModify="((Item=(Type=Bone,Name=\"dyn_skirt_b_r_base\"),Transform=(Rotation=(X=0,Y=-0.2588,Z=0,W=0.9659),Translation=(X=0,Y=5,Z=5),Scale3D=(X=1,Y=1,Z=1))),(Item=(Type=Bone,Name=\"dyn_skirt_s_l_base\"),Transform=(Rotation=(X=0,Y=-0.0872,Z=0,W=0.9962),Translation=(X=0,Y= ...(truncated)"
    WeightMinimum=0
    Weight=RigVMVariableNode() {variable=L thighBack Alpha}.Value
   )
   n22 = RigVMVariableNode() {variable=R calfBend Alpha}
   n23 = RigVMDispatchNode(False=(X=0,Y=-0.5,Z=0,W=0.866), True=(X=0,Y=-0.1736,Z=0,W=0.9848), Condition=RigVMVariableNode() {variable=isCapeWearing}.Value) {template="DISPATCH_RigVMDispatch_If(in Condition,in True,in False,out Result)"}
   > &RigUnit_ModifyTransforms(
    ItemToModify="((Item=(Type=Bone,Name=\"dyn_skirt_b_r_03\"),Transform=(Rotation=(X=0,Y=-0.8192,Z=0,W=0.5736),Translation=(X=0,Y=0,Z=0),Scale3D=(X=1,Y=1,Z=1))),(Item=(Type=Bone,Name=\"dyn_skirt_s_r_03\"),Transform=(Rotation=(X=0,Y=-0.8192,Z=0,W=0.5736),Translation=(X=0,Y=0,Z= ...(truncated)"
    WeightMinimum=0
    Weight=RigVMDispatchNode(True=0.3, False=@n22.Value, Condition=RigVMVariableNode() {variable=isEmote}.Value) {template="DISPATCH_RigVMDispatch_If(in Condition,in True,in False,out Result)"}.Result
    ItemToModify.3.Transform.Rotation=@n23.Result
   )
   n24 = RigVMVariableNode() {variable=L calfBend Alpha}
   > &RigUnit_ModifyTransforms(
    ItemToModify="((Item=(Type=Bone,Name=\"dyn_skirt_b_l_03\"),Transform=(Rotation=(X=0,Y=-0.8192,Z=0,W=0.5736),Translation=(X=0,Y=0,Z=0),Scale3D=(X=1,Y=1,Z=1))),(Item=(Type=Bone,Name=\"dyn_skirt_s_l_03\"),Transform=(Rotation=(X=0,Y=-0.8192,Z=0,W=0.5736),Translation=(X=0,Y=0,Z= ...(truncated)"
    WeightMinimum=0
    Weight=RigVMDispatchNode(True=0.3, Condition=RigVMVariableNode() {variable=isEmote}.Value, False=@n24.Value) {template="DISPATCH_RigVMDispatch_If(in Condition,in True,in False,out Result)"}.Result
    ItemToModify.3.Transform.Rotation=@n23.Result
   )
   n25 = RigVMVariableNode() {variable=sideToside}
   > &RigUnit_ModifyTransforms(
    ItemToModify="((Item=(Type=Bone,Name=\"dyn_skirt_f2_l_base\"),Transform=(Rotation=(X=0.0189,Y=-0.0851,Z=0.2156,W=0.9726),Translation=(X=0,Y=0,Z=0),Scale3D=(X=1,Y=1,Z=1))),(Item=(Type=Bone,Name=\"dyn_skirt_f_c_base\"),Transform=(Rotation=(X=0.05,Y=-0.1206,Z=0.3794,W=0.916),T ...(truncated)"
    WeightMinimum=0
    Weight=&RigVMFunction_MathDoubleMul(B=@n25.Value, A=RigVMVariableNode() {variable=L thighSideIn_Fwd Alpha}.Value).Result
   )
   > &RigUnit_ModifyTransforms(
    ItemToModify="((Item=(Type=Bone,Name=\"dyn_skirt_f2_r_base\"),Transform=(Rotation=(X=-0.0189,Y=-0.0851,Z=-0.2156,W=0.9726),Translation=(X=0,Y=0,Z=0),Scale3D=(X=1,Y=1,Z=1))),(Item=(Type=Bone,Name=\"dyn_skirt_f_c_base\"),Transform=(Rotation=(X=-0.05,Y=-0.1206,Z=-0.3794,W=0.91 ...(truncated)"
    WeightMinimum=0
    Weight=&RigVMFunction_MathDoubleMul(B=@n25.Value, A=RigVMVariableNode() {variable=R thighSideIn_Fwd Alpha}.Value).Result
   )
   > &RigUnit_ModifyTransforms(
    ItemToModify="((Item=(Type=Bone,Name=\"dyn_skirt_b_r_base\"),Transform=(Rotation=(X=-0.0468,Y=-0.2113,Z=-0.2113,W=0.9532),Translation=(X=0,Y=0,Z=0),Scale3D=(X=1,Y=1,Z=1))),(Item=(Type=Bone,Name=\"dyn_skirt_b_l_base\"),Transform=(Rotation=(X=-0.0283,Y=-0.1274,Z=-0.2146,W=0.9 ...(truncated)"
    WeightMinimum=0
    Weight=&RigVMFunction_MathDoubleMul(B=@n25.Value, A=RigVMVariableNode() {variable=L thighSideIn_Bwd Alpha}.Value).Result
   )
   > &RigUnit_ModifyTransforms(
    ItemToModify="((Item=(Type=Bone,Name=\"dyn_skirt_b_r_base\"),Transform=(Rotation=(X=0.0189,Y=-0.2156,Z=0.0851,W=0.9726),Translation=(X=0,Y=0,Z=0),Scale3D=(X=1,Y=1,Z=1))),(Item=(Type=Bone,Name=\"dyn_skirt_b_l_base\"),Transform=(Rotation=(X=-0.0468,Y=0.2113,Z=0.2113,W=0.9532) ...(truncated)"
    WeightMinimum=0
    Weight=&RigVMFunction_MathDoubleMul(B=@n25.Value, A=RigVMVariableNode() {variable=R thighSideIn_Bwd Alpha}.Value).Result
   )
   RigVMFunction_AnimEvalRichCurve = &RigVMFunction_AnimEvalRichCurve(
    Curve=(EditorCurveData=(Keys=((Time=0.0021,Value=0.0154),(InterpMode=RCIM_Cubic,Time=0.4113,Value=0.6732,ArriveTangent=0.9867,LeaveTangent=0.9867),(InterpMode=RCIM_Cubic,TangentMode=RCTM_Break,Time=1,Value=1,ArriveTangent=-0.4632))))
    SourceMinimum=0
    TargetMinimum=0
    Value=RigVMVariableNode() {variable=crouchFloa2}.Value
   )
   Interpolate = /Script/RigVM.RigVMFunction_MathQuaternionSlerp(A=(X=0.056,Y=-0.6403,Z=0.0668,W=0.7631), B=(X=0,Y=-0.9848,Z=0,W=0.1736), T=@RigVMFunction_AnimEvalRichCurve.Result) {is=struct RigVMUnitNode, template="Interpolate::Execute(in A,in B,in T,out Result)"}
   &RigUnit_ModifyTransforms(
    ItemToModify="((Item=(Type=Bone,Name=\"dyn_skirt_b_r_base\"),Transform=(Rotation=(X=0,Y=-0.9397,Z=0,W=0.342),Translation=(X=20,Y=0,Z=0),Scale3D=(X=0.5,Y=1,Z=1))),(Item=(Type=Bone,Name=\"dyn_skirt_s_r_base\"),Transform=(Rotation=(X=0,Y=-0.0872,Z=0,W=0.9962),Translation=(X=0, ...(truncated)"
    WeightMinimum=0
    ItemToModify.0.Transform.Rotation=@Interpolate.Result
    Weight=&RigVMFunction_MathDoubleMul(B=RigVMVariableNode() {variable=crouchFloat}.Value, A=RigVMVariableNode() {variable=R calfBend Alpha}.Value).Result
    ExecutePin=@n21.True
   )
   > &RigUnit_ModifyTransforms(
    ItemToModify="((Item=(Type=Bone,Name=\"dyn_skirt_b_l_base\"),Transform=(Rotation=(X=0,Y=-0.7071,Z=0,W=0.7071),Translation=(X=0,Y=0,Z=0),Scale3D=(X=0.5,Y=1,Z=1))),(Item=(Type=Bone,Name=\"dyn_skirt_s_l_base\"),Transform=(Rotation=(X=0,Y=-0.2164,Z=0,W=0.9763),Translation=(X=0, ...(truncated)"
    WeightMinimum=0
    ItemToModify.0.Transform.Rotation=&RigVMFunction_MathQuaternionSlerp(A=(X=0,Y=0,Z=0,W=1), B=(X=0,Y=-0.7071,Z=0,W=0.7071), T=@RigVMFunction_AnimEvalRichCurve.Result) {template="Interpolate::Execute(in A,in B,in T,out Result)"}.Result
    Weight=&RigVMFunction_MathDoubleMul(A=RigVMVariableNode() {variable=L calfBend Alpha}.Value, B=RigVMVariableNode() {variable=crouchFloat}.Value).Result
   )
   > &RigUnit_ParentConstraint(
    AdvancedSettings=(InterpolationType=Average,RotationOrderForFilter=XZY)
    Child=(Type=Bone,Name="dyn_skirt_f_r_03")
    Parents=((Item=(Type=Bone,Name="calf_r"),Weight=1))
    Weight=1
    bMaintainOffset=true
   )
   > &RigUnit_ParentConstraint(
    AdvancedSettings=(InterpolationType=Average,RotationOrderForFilter=XZY)
    Child=(Type=Bone,Name="dyn_skirt_f_l_03")
    Parents=((Item=(Type=Bone,Name="calf_l"),Weight=1))
    Weight=1
    bMaintainOffset=true
   )
   > &RigUnit_ParentConstraint(
    AdvancedSettings=(InterpolationType=Average,RotationOrderForFilter=XZY)
    Child=(Type=Bone,Name="dyn_skirt_f2_r_03")
    Parents=((Item=(Type=Bone,Name="calf_r"),Weight=1))
    Weight=0.3
    bMaintainOffset=true
   )
   > &RigUnit_ParentConstraint(
    AdvancedSettings=(InterpolationType=Average,RotationOrderForFilter=XZY)
    Child=(Type=Bone,Name="dyn_skirt_f2_l_03")
    Parents=((Item=(Type=Bone,Name="calf_l"),Weight=1))
    Weight=1
    bMaintainOffset=true
   )
   > &RigUnit_ParentConstraint(
    AdvancedSettings=(InterpolationType=Average,RotationOrderForFilter=XZY)
    Child=(Type=Bone,Name="dyn_skirt_b_l_03")
    Parents=((Item=(Type=Bone,Name="foot_l"),Weight=1))
    Weight=0.2
    bMaintainOffset=true
   )
   > &RigUnit_ParentConstraint(
    AdvancedSettings=(InterpolationType=Average,RotationOrderForFilter=XZY)
    Child=(Type=Bone,Name="dyn_skirt_s_l_02")
    Parents=((Item=(Type=Bone,Name="thigh_l"),Weight=1))
    Weight=0.2
    bMaintainOffset=true
   )
   > &RigUnit_ParentConstraint(
    AdvancedSettings=(InterpolationType=Average,RotationOrderForFilter=XZY)
    Child=(Type=Bone,Name="dyn_skirt_f_c_base")
    Parents=((Item=(Type=Bone,Name="thigh_l"),Weight=1))
    Weight=0.3
    bMaintainOffset=true
   )
   > &RigUnit_ModifyTransforms(
    ItemToModify="((Item=(Type=Bone,Name=\"dyn_skirt_s_r_02\"),Transform=(Rotation=(X=0,Y=0,Z=0.0872,W=0.9962),Translation=(X=0,Y=0,Z=0),Scale3D=(X=1,Y=1,Z=1))),(Item=(Type=Bone,Name=\"dyn_skirt_f2_r_02\"),Transform=(Rotation=(X=-0.0094,Y=-0.0426,Z=-0.2162,W=0.9754),Translation ...(truncated)"
    Weight=1
    WeightMinimum=0
   )
   &RigUnit_ModifyTransforms(
    ExecutePin=()
    ItemToModify=((Item=(Type=Bone,Name="dyn_skirt_b_r_01"),Transform=(Rotation=(X=0,Y=0,Z=0,W=1),Translation=(X=0,Y=1,Z=0),Scale3D=(X=1,Y=1,Z=1))))
    Weight=1
    WeightMinimum=0
   )
   &RigUnit_ModifyTransforms(
    ExecutePin=()
    ItemToModify=((Item=(Type=Bone,Name="dyn_skirt_b_l_02"),Transform=(Rotation=(X=0,Y=-0.0872,Z=0,W=0.9962),Translation=(X=0,Y=0,Z=6),Scale3D=(X=1,Y=1,Z=1))))
    Weight=1
    WeightMinimum=0
   )
   &RigUnit_ParentConstraint(
    AdvancedSettings=(InterpolationType=Average,RotationOrderForFilter=XZY)
    Child=(Type=Bone,Name="dyn_skirt_b_r_03")
    ExecutePin=()
    Parents=((Item=(Type=Bone,Name="foot_r"),Weight=1))
    Weight=0.3
    bMaintainOffset=true
   )
   &RigVMFunction_MathDoubleSub(
    B=&RigVMFunction_MathDoubleRemap(SourceMaximum=1, SourceMinimum=0, TargetMaximum=3, TargetMinimum=0, Value=RigVMVariableNode() {variable=R thighForward Alpha}.Value).Result
    A=@n22.Value
   )
   &RigVMFunction_MathDoubleSub(
    B=&RigVMFunction_MathDoubleRemap(SourceMaximum=1, SourceMinimum=0, TargetMaximum=3, TargetMinimum=0, Value=RigVMVariableNode() {variable=L thighForward Alpha}.Value).Result
    A=@n24.Value
   )
   RigVMVariableNode() {variable=pawnSpeed}
  ]}
 New Function_ContainedGraph {nodes [
   RigVMFunctionEntryNode()
   RigVMFunction_MathQuaternionToEuler = &RigVMFunction_MathQuaternionToEuler(
    RotationOrder=ZYX
    Value=&RigVMFunction_MathQuaternionSwingTwist(
     TwistAxis=(X=1,Y=0,Z=0)
     Input=&RigUnit_GetRelativeTransformForItem(Child=(Type=Bone,Name="head"), Parent=(Type=Control,Name="ref_head_ctrl"), bChildInitial=False, bParentInitial=False).RelativeTransform.Rotation
    ).Swing
   )
   n1 = &RigVMFunction_MathQuaternionToEuler(
    RotationOrder=ZYX
    Value=&RigVMFunction_MathQuaternionSwingTwist(
     TwistAxis=(X=1,Y=0,Z=0)
     Input=&RigUnit_GetRelativeTransformForItem(Child=(Type=Bone,Name="neck_02"), Parent=(Type=Control,Name="ref_neck_ctrl"), bChildInitial=False, bParentInitial=False).RelativeTransform.Rotation
    ).Swing
   )
   Maximum_2 = RigVMAggregateNode(
    B=&RigVMFunction_MathDoubleRemap(SourceMaximum=120, SourceMinimum=0, TargetMaximum=1, TargetMinimum=0, Value=@RigVMFunction_MathQuaternionToEuler.Result.Y).Result
    C=&RigVMFunction_MathDoubleRemap(SourceMaximum=40, SourceMinimum=0, TargetMaximum=1, TargetMinimum=0, Value=@n1.Result.Y).Result
    A=&RigVMFunction_MathDoubleRemap(SourceMaximum=-70, SourceMinimum=0, TargetMaximum=1, TargetMinimum=0, Value=RigVMVariableNode() {variable=rotation_l}.Value.Y).Result
   )
   > RigVMVariableNode(Value=@Maximum_2.Result) {variable=head_side_l}
   n2 = &RigVMFunction_MathDoubleRemap(SourceMaximum=-70, SourceMinimum=0, TargetMaximum=1, TargetMinimum=0, Value=RigVMVariableNode() {variable=rotation_r}.Value.Y)
   Add = RigVMAggregateNode(
    B=&RigVMFunction_MathDoubleRemap(SourceMaximum=-120, SourceMinimum=0, TargetMaximum=1, TargetMinimum=0, Value=@RigVMFunction_MathQuaternionToEuler.Result.Y).Result
    C=&RigVMFunction_MathDoubleRemap(SourceMaximum=-40, SourceMinimum=0, TargetMaximum=1, TargetMinimum=0, Value=@n1.Result.Y).Result
    A=@n2.Result
   )
   Maximum_1 = RigVMAggregateNode(B=@Add.Result, A=@n2.Result)
   > RigVMVariableNode(Value=@Maximum_1.Result) {variable=head_side_r}
   RigVMFunction_AnimEasing = /Script/RigVM.RigVMFunction_AnimEasing(
    SourceMaximum=1
    SourceMinimum=0
    TargetMaximum=1
    TargetMinimum=0
    Type=CubicEaseIn
    Value=&RigVMFunction_MathDoubleRemap(SourceMaximum=-100, SourceMinimum=0, TargetMaximum=1, TargetMinimum=0, Value=@RigVMFunction_MathQuaternionToEuler.Result.Z).Result
   ) {is=struct RigVMUnitNode}
   > RigVMVariableNode(Value=&RigVMFunction_MathDoubleAdd(
     B=&RigVMFunction_MathDoubleRemap(SourceMaximum=-100, SourceMinimum=0, TargetMaximum=1, TargetMinimum=0, Value=@n1.Result.Z).Result
     A=@RigVMFunction_AnimEasing.Result
    ).Result) {variable=head_back_c}
   n3 = &RigVMFunction_MathDoubleRemap(SourceMaximum=100, SourceMinimum=0, TargetMaximum=1, TargetMinimum=0, Value=@n1.Result.Y)
   > RigVMVariableNode(Value=&RigVMFunction_MathDoubleAdd(
     B=&RigVMFunction_MathDoubleRemap(SourceMaximum=100, SourceMinimum=0, TargetMaximum=1, TargetMinimum=0, Value=@n1.Result.Z).Result
     A=&RigVMFunction_AnimEasing(SourceMaximum=1, SourceMinimum=0, TargetMaximum=1, TargetMinimum=0, Type=CubicEaseIn, Value=@n3.Result).Result
    ).Result) {variable=HeadFr}
   > RigVMFunctionReturnNode()
  ]}
 RigVMFunctionLibrary {nodes [
   DeltaRotY = RigVMCollapseNode(ExecuteContext=())
   Neck Collar Get Variables = RigVMCollapseNode(ExecuteContext=())
   Skirt_MG_dress_setup = RigVMCollapseNode(ExecuteContext=())
   VarAndConstraints = RigVMCollapseNode(ExecuteContext=())
   isSliding = RigVMCollapseNode(ExecuteContext=())
  ]}
 RigVMFunction_Sequence_1_ContainedGraph {nodes [
   n1 = RigVMFunctionEntryNode()
   RigVMFunction_Sequence = &RigVMFunction_Sequence(ExecuteContext=@n1.ExecuteContext)
   n2 = &RigVMFunction_Sequence(ExecuteContext=@RigVMFunction_Sequence.B)
   RigVMFunctionReturnNode(A=@RigVMFunction_Sequence.A, B=@n2.A, C=@n2.B)
  ]}
 RigVMFunction_Sequence_1_ContainedGraph {nodes [
   n1 = RigVMFunctionEntryNode()
   RigVMFunction_Sequence = &RigVMFunction_Sequence(ExecuteContext=@n1.ExecuteContext)
   n2 = &RigVMFunction_Sequence(ExecuteContext=@RigVMFunction_Sequence.B)
   n3 = &RigVMFunction_Sequence(ExecuteContext=@n2.B)
   n4 = &RigVMFunction_Sequence(ExecuteContext=@n3.B)
   n5 = &RigVMFunction_Sequence(ExecuteContext=@n4.B)
   n6 = &RigVMFunction_Sequence(ExecuteContext=@n5.B)
   n7 = &RigVMFunction_Sequence(ExecuteContext=@n6.B)
   n8 = &RigVMFunction_Sequence(ExecuteContext=@n7.B)
   RigVMFunctionReturnNode(A=@RigVMFunction_Sequence.A, B=@n2.A, C=@n3.A, D=@n4.A, E=@n5.A, F=@n6.A, G=@n7.A, H=@n8.A, I=@n8.B)
  ]}
 RigVMModel {nodes [
   n1 = &RigUnit_PrepareForExecution()
   &RigUnit_HierarchyImportFromSkeleton(NameSpace=None, bIncludeCurves=False, bIncludeMeshSockets=False, bIncludeVirtualBones=True, ExecutePin=@n1.ExecutePin)
   > &RigUnit_HierarchyAddNull(
    Name=Ref_Thigh_L
    Parent=(Type=Bone,Name="pelvis")
    Space=GlobalSpace
    Transform=&RigUnit_GetTransform(Item=(Type=Bone,Name="thigh_twist_01_l"), bInitial=False).Transform
   )
   > &RigUnit_HierarchyAddNull(
    Name=Ref_Thigh_R
    Parent=(Type=Bone,Name="pelvis")
    Space=GlobalSpace
    Transform=&RigUnit_GetTransform(Item=(Type=Bone,Name="thigh_twist_01_r"), bInitial=False).Transform
   )
   > Control Stack at Item = RigVMFunctionReferenceNode(
    Add Bottom Null=False
    Add Null=true
    Add Secondary Control=False
    Control Suffix=_ctrl
    Item=(Name="head",Type=Bone)
    Name=ref_head
    Null Suffix=_null
    Parent=(Name="neck_02",Type=Bone)
    Secondary Settings="(Shape=(bVisible=True,Name=\"Default\",Color=(R=1,G=0,B=0,A=1),Transform=(Rotation=(X=0,Y=0,Z=0,W=1),Translation=(X=0,Y=0,Z=0),Scale3D=(X=1,Y=1,Z=1))),Proxy=(bIsProxy=False,ShapeVisibility=BasedOnSelection),DisplayName=\"None\",InitialSpace=LocalSpace,bUsePref ...(truncated)"
    Settings="(Shape=(bVisible=True,Name=\"Pyramid_Thin\",Color=(R=1,G=0,B=0,A=1),Transform=(Rotation=(X=0,Y=0,Z=0,W=1),Translation=(X=0,Y=0,Z=0),Scale3D=(X=1,Y=1,Z=1))),Proxy=(bIsProxy=False,ShapeVisibility=BasedOnSelection),DisplayName=\"None\",InitialSpace=LocalSpace,bUs ...(truncated)"
   )
   > Control Stack at Item_1 = RigVMFunctionReferenceNode(
    Add Bottom Null=False
    Add Null=true
    Add Secondary Control=False
    Control Suffix=_ctrl
    Item=(Name="neck_02",Type=Bone)
    Name=ref_neck
    Null Suffix=_null
    Parent=(Name="spine_05",Type=Bone)
    Secondary Settings="(Shape=(bVisible=True,Name=\"Default\",Color=(R=1,G=0,B=0,A=1),Transform=(Rotation=(X=0,Y=0,Z=0,W=1),Translation=(X=0,Y=0,Z=0),Scale3D=(X=1,Y=1,Z=1))),Proxy=(bIsProxy=False,ShapeVisibility=BasedOnSelection),DisplayName=\"None\",InitialSpace=LocalSpace,bUsePref ...(truncated)"
    Settings="(Shape=(bVisible=True,Name=\"Pyramid_Thin\",Color=(R=1,G=0,B=0,A=1),Transform=(Rotation=(X=0,Y=0,Z=0,W=1),Translation=(X=0,Y=0,Z=0),Scale3D=(X=1,Y=1,Z=1))),Proxy=(bIsProxy=False,ShapeVisibility=BasedOnSelection),DisplayName=\"None\",InitialSpace=LocalSpace,bUs ...(truncated)"
   )
   > Control Stack at Item_2 = RigVMFunctionReferenceNode(
    Add Bottom Null=False
    Add Null=true
    Add Secondary Control=False
    Control Suffix=_ctrl
    Item=(Name="clavicle_l",Type=Bone)
    Name=ref_clavicle_l
    Null Suffix=_null
    Parent=(Name="spine_05",Type=Bone)
    Secondary Settings="(Shape=(bVisible=True,Name=\"Default\",Color=(R=1,G=0,B=0,A=1),Transform=(Rotation=(X=0,Y=0,Z=0,W=1),Translation=(X=0,Y=0,Z=0),Scale3D=(X=1,Y=1,Z=1))),Proxy=(bIsProxy=False,ShapeVisibility=BasedOnSelection),DisplayName=\"None\",InitialSpace=LocalSpace,bUsePref ...(truncated)"
    Settings="(Shape=(bVisible=True,Name=\"Pyramid_Thin\",Color=(R=1,G=0,B=0,A=1),Transform=(Rotation=(X=0,Y=0,Z=0,W=1),Translation=(X=0,Y=0,Z=0),Scale3D=(X=1,Y=1,Z=1))),Proxy=(bIsProxy=False,ShapeVisibility=BasedOnSelection),DisplayName=\"None\",InitialSpace=LocalSpace,bUs ...(truncated)"
   )
   > Control Stack at Item_3 = RigVMFunctionReferenceNode(
    Add Bottom Null=False
    Add Null=true
    Add Secondary Control=False
    Control Suffix=_ctrl
    Item=(Name="clavicle_r",Type=Bone)
    Name=ref_clavicle_r
    Null Suffix=_null
    Parent=(Name="spine_05",Type=Bone)
    Secondary Settings="(Shape=(bVisible=True,Name=\"Default\",Color=(R=1,G=0,B=0,A=1),Transform=(Rotation=(X=0,Y=0,Z=0,W=1),Translation=(X=0,Y=0,Z=0),Scale3D=(X=1,Y=1,Z=1))),Proxy=(bIsProxy=False,ShapeVisibility=BasedOnSelection),DisplayName=\"None\",InitialSpace=LocalSpace,bUsePref ...(truncated)"
    Settings="(Shape=(bVisible=True,Name=\"Pyramid_Thin\",Color=(R=1,G=0,B=0,A=1),Transform=(Rotation=(X=0,Y=0,Z=0,W=1),Translation=(X=0,Y=0,Z=0),Scale3D=(X=1,Y=1,Z=1))),Proxy=(bIsProxy=False,ShapeVisibility=BasedOnSelection),DisplayName=\"None\",InitialSpace=LocalSpace,bUs ...(truncated)"
   )
   > &RigUnit_HierarchyAddNull(
    Name=thigh_l_null
    Parent=(Type=Bone,Name="pelvis")
    Space=GlobalSpace
    Transform=&RigUnit_GetTransform(Item=(Type=Bone,Name="thigh_l"), bInitial=false).Transform
   )
   > &RigUnit_HierarchyAddNull(
    Name=thigh_r_null
    Parent=(Type=Bone,Name="pelvis")
    Space=GlobalSpace
    Transform=&RigUnit_GetTransform(Item=(Type=Bone,Name="thigh_r"), bInitial=false).Transform
   )
   > &RigUnit_HierarchyAddNull(
    Name=frontSkirt_01_null
    Parent=(Type=Bone,Name="dyn_frontSkirt_01")
    Space=LocalSpace
    Transform=(Rotation=(X=0,Y=0,Z=0,W=1),Translation=(X=0,Y=-10,Z=0),Scale3D=(X=1,Y=1,Z=1))
   )
   > &RigUnit_HierarchyAddNull(
    Name=frontSkirt_02_null
    Parent=(Type=Bone,Name="dyn_frontSkirt_02")
    Space=LocalSpace
    Transform=(Rotation=(X=0,Y=0,Z=0,W=1),Translation=(X=1,Y=-10,Z=0),Scale3D=(X=1,Y=1,Z=1))
   )
   > &RigUnit_HierarchyAddNull(
    Name=frontSkirt_01_l_null
    Parent=(Type=Bone,Name="dyn_frontSkirt_01_l")
    Space=LocalSpace
    Transform=(Rotation=(X=0,Y=0,Z=0,W=1),Translation=(X=0,Y=2,Z=0.5),Scale3D=(X=1,Y=1,Z=1))
   )
   > &RigUnit_HierarchyAddNull(
    Name=frontSkirt_02_l_null
    Parent=(Type=Bone,Name="dyn_frontSkirt_02_l")
    Space=LocalSpace
    Transform=(Rotation=(X=0,Y=0,Z=0,W=1),Translation=(X=0,Y=1,Z=1),Scale3D=(X=1,Y=1,Z=1))
   )
   > &RigUnit_HierarchyAddNull(
    Name=calf_l_null
    Parent=(Type=Bone,Name="calf_l")
    Space=LocalSpace
    Transform=(Rotation=(X=0,Y=0,Z=0,W=1),Translation=(X=0,Y=12.5,Z=0),Scale3D=(X=1,Y=1,Z=1))
   )
   RigUnit_BeginExecution = &RigUnit_BeginExecution()
   n2 = &RigVMFunction_Sequence(ExecuteContext=@RigUnit_BeginExecution.ExecutePin)
   FN_DeformRig_FNC_1 = RigVMFunctionReferenceNode(
    DeformRig_ Curve="(deformRig_l_upperarm_up_alpha_72_#1=true,deformRig_r_upperarm_up_alpha_73_#2=true,deformRig_l_upperarm_fr_alpha_74_#3=true,deformRig_r_upperarm_fr_alpha_75_CB51B80C4 ...(truncated)"
    DeformRig_Bool="(upperarm_1_#4=True,elbow_3_#5=True,wrist_7_#6=True,groin_8_#7=True,glute_17_#8=True,knee_9_40C408324CA261 ...(truncated)"
    DeformRig_Debug=(upperarm_debug_18_#4=False,wrist_debug_20_#6=False,lat_debug_26_#9=False,groinGlute_debug_27_#7=False)
    DeformRig_Value="(upperarm_up_18_#10=(Rotation=(X=0,Y=0,Z=0,W=1),Translation=(X=0,Y=0,Z=8),Scale3D=(X=0.7,Y=1,Z=1)),upperarm_fwd_fr_19_#11=(Rotation=(X=0,Y=0,Z=0,W=1),Translation=(X=0,Y=10,Z=0),Scale3D=(X=0.5,Y=1,Z= ...(truncated)"
    ExecuteContext=@n2.A
   )
   > VarAndConstraints = RigVMFunctionReferenceNode()
   > Skirt_MG_dress_setup = RigVMFunctionReferenceNode()
   > Neck Collar Get Variables = RigVMFunctionReferenceNode()
   > &RigUnit_ModifyTransforms(
    ItemToModify="((Item=(Type=Bone,Name=\"hood_side_l\"),Transform=(Rotation=(X=0,Y=0,Z=0.3827,W=0.9239),Translation=(X=0,Y=5,Z=0),Scale3D=(X=1,Y=1,Z=1))),(Item=(Type=Bone,Name=\"hood_front_l\"),Transform=(Rotation=(X=0,Y=0,Z=0.3827,W=0.9239),Translation=(X=0,Y=0,Z=0),Scale3D= ...(truncated)"
    WeightMinimum=0
    Weight=RigVMVariableNode() {variable=head_side_l}.Value
   )
   > &RigUnit_ModifyTransforms(
    ItemToModify="((Item=(Type=Bone,Name=\"hood_side_r\"),Transform=(Rotation=(X=0.099,Y=-0.2391,Z=0.3696,W=0.8924),Translation=(X=0,Y=-5,Z=0),Scale3D=(X=1,Y=1,Z=1))),(Item=(Type=Bone,Name=\"hood_front_r\"),Transform=(Rotation=(X=0,Y=0,Z=0.3827,W=0.9239),Translation=(X=0,Y=0,Z= ...(truncated)"
    WeightMinimum=0
    Weight=RigVMVariableNode() {variable=head_side_r}.Value
   )
   > &RigUnit_ModifyTransforms(
    ItemToModify="((Item=(Type=Bone,Name=\"hood_back_c\"),Transform=(Rotation=(X=0,Y=0,Z=-1,W=0),Translation=(X=0,Y=-6,Z=9),Scale3D=(X=0.2,Y=1,Z=1))),(Item=(Type=Bone,Name=\"hood_back_side_r\"),Transform=(Rotation=(X=0,Y=0,Z=0.6428,W=0.766),Translation=(X=0,Y=-7,Z=0),Scale3D=(X ...(truncated)"
    WeightMinimum=0
    Weight=RigVMVariableNode() {variable=head_back_c}.Value
   )
   > &RigUnit_ModifyTransforms(
    ItemToModify="((Item=(Type=Bone,Name=\"hood_front_l\"),Transform=(Rotation=(X=0,Y=0,Z=0,W=1),Translation=(X=6,Y=9,Z=15),Scale3D=(X=1,Y=1,Z=1))),(Item=(Type=Bone,Name=\"hood_front_r\"),Transform=(Rotation=(X=0,Y=0,Z=0,W=1),Translation=(X=6,Y=-9,Z=-15),Scale3D=(X=1,Y=1,Z=1))) ...(truncated)"
    WeightMinimum=0
    Weight=RigVMVariableNode() {variable=HeadFr}.Value
   )
   > &RigUnit_ModifyTransforms(
    ItemToModify=((Item=(Type=Bone,Name="base_upperarm_twist_01_extra_l"),Transform=(Rotation=(X=0,Y=0.0872,Z=0,W=0.9962),Translation=(X=3,Y=0,Z=1),Scale3D=(X=0.9,Y=0.9,Z=1))))
    WeightMinimum=0
    Weight=RigVMDispatchNode(BoneName=root, Default=0, Name=deformRig_l_upperarm_up_alpha) {template="DISPATCH_RigDispatch_GetAnimAttribute(in Name,in BoneName,in Default,out Value,out Found)"}.Value
   )
   isSliding = RigVMFunctionReferenceNode(ExecuteContext=@n2.B)
   n3 = &RigUnit_GetRelativeTransformForItem(Child=(Type=Bone,Name="clavicle_l"), Parent=(Type=Control,Name="ref_clavicle_l_ctrl"), bChildInitial=False, bParentInitial=False)
   RigVMVariableNode(ExecuteContext=(), Value=&RigVMFunction_MathQuaternionToEuler(RotationOrder=ZYX, Value=@n3.RelativeTransform.Rotation).Result) {variable=rotation_l}
   n4 = &RigUnit_GetRelativeTransformForItem(Child=(Type=Bone,Name="clavicle_r"), Parent=(Type=Control,Name="ref_clavicle_r_ctrl"), bChildInitial=False, bParentInitial=False)
   > RigVMVariableNode(Value=&RigVMFunction_MathQuaternionToEuler(RotationOrder=ZYX, Value=@n4.RelativeTransform.Rotation).Result) {variable=rotation_r}
  ]}
]
variables [
 HeadFr {type=double, default=0}
 Is Crouching_0 {type=bool, default=False}
 L Crouch Strafe Behind {type=double, default=0}
 L Crouch Strafe Behind_0 {type=double, default=0}
 L Crouch Strafe Front {type=double, default=0}
 L Crouch Strafe Front_0 {type=double, default=0}
 L calfBend Alpha {type=double, default=0}
 L thighBack Alpha {type=double, default=0}
 L thighForward Alpha {type=double, default=0}
 L thighSide Alpha {type=double, default=0}
 L thighSideIn_Bwd Alpha {type=double, default=0}
 L thighSideIn_Fwd Alpha {type=double, default=0}
 L thighTwistIn Alpha {type=double, default=0}
 L thighTwistOut Alpha {type=double, default=0}
 Not Skydive Alpha {type=double, default=0}
 R Crouch Strafe Behind {type=double, default=0}
 R Crouch Strafe Behind_0 {type=double, default=0}
 R Crouch Strafe Front {type=double, default=0}
 R Crouch Strafe Front_0 {type=double, default=0}
 R calfBend Alpha {type=double, default=0}
 R thighBack Alpha {type=double, default=0}
 R thighForward Alpha {type=double, default=0}
 R thighSide Alpha {type=double, default=0}
 R thighSideIn_Bwd Alpha {type=double, default=0}
 R thighSideIn_Fwd Alpha {type=double, default=0}
 R thighTwistIn Alpha {type=double, default=0}
 ThighSplit L Fwd {type=double, default=0}
 ThighSplit R Fwd {type=double, default=0}
 Thigh_L_Side {type=double, default=0}
 Thigh_R_Side {type=double, default=0}
 crouchFloa2 {type=double, default=0}
 crouchFloat {type=double, default=0}
 head_back_c {type=double, default=0}
 head_side_l {type=double, default=0}
 head_side_r {type=double, default=0}
 is crouching {type=bool, default=False}
 isCapeWearing {type=bool, default=False}
 isEmote {type=bool, default=False}
 isSlidingVar_1 {type=bool, default=False}
 isSlidingVar_2 {type=bool, default=False}
 leg up crouching {type=double, default=0}
 leg up crouching_r {type=double, default=0}
 leg up out crouching {type=double, default=0}
 leg up out crouching_r {type=double, default=0}
 leg_back_l {type=double, default=0}
 leg_out_l {type=double, default=0}
 leg_out_r {type=double, default=0}
 leg_up_l {type=double, default=0}
 leg_up_r {type=double, default=0}
 pawnSpeed {type=double, default=0}
 rotation_l {type=FVector, default=(X=0,Y=0,Z=0)}
 rotation_r {type=FVector, default=(X=0,Y=0,Z=0)}
 sideToside {type=double, default=0}
]
