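# Physics asset for the F_MED_SableHorizon female body cosmetic, set up for secondary motion of its skirt, front belt pieces and collar.
# Nine four-link skirt chains and three belt pieces hang from pelvis anchors, and a two-link collar from spine_05.
# Torso and leg capsules provide body collision, while a huge root capsule below the origin likely acts as a floor.
# Dynamic joints have tight twist limits, with one swing axis widening down the skirt.
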PhysicsAsset /BRCosmetics/Characters/Player/Female/Medium/Bodies/F_MED_SableHorizon/Meshes/F_MED_SableHorizon_Physics
bodies {
 base_beltfront [{shapeType=Capsule, length=0.505, radius=0.505, shapeName=base_beltfront_capsule, rotation=<rot 0 -90 0>}]
 base_beltfront_l [{shapeType=Capsule, length=0.505, radius=0.505, shapeName=base_beltfront_l_capsule, rotation=<rot 0 -90 0>}]
 base_beltfront_r [{shapeType=Capsule, length=0.505, radius=0.505, shapeName=base_beltfront_r_capsule, rotation=<rot 0 -90 0>}]
 base_collar [{shapeType=Capsule, length=0.505, radius=0.505, shapeName=base_collar_capsule, rotation=<rot 0 -90 0>}]
 calf_l [{shapeType=Capsule, length=30.4353, radius=9, shapeName=calf_l_capsule, center=<-18.9388 -2.4817 -0.0002>, rotation=<rot -90 0 0>}]
 calf_r [{shapeType=Capsule, length=30.4353, radius=9, shapeName=calf_r_capsule, center=<18.2458 2.4817 0.0016>, rotation=<rot -90 0 0>}]
 dyn_beltfront [{shapeType=Capsule, length=2, radius=3, shapeName=dyn_beltfront_capsule, center=<3 0 0>, rotation=<rot 90 90 0>}]
 dyn_beltfront_l [{shapeType=Capsule, length=2, radius=3, shapeName=dyn_beltfront_l_capsule, center=<2 0 0>, rotation=<rot 90 90 0>}]
 dyn_beltfront_r [{shapeType=Capsule, length=2, radius=3, shapeName=dyn_beltfront_r_capsule, center=<3 0 0>, rotation=<rot 90 90 0>}]
 dyn_collar_01 [{shapeType=Capsule, length=2, radius=1, shapeName=dyn_collar_01_capsule, center=<2 0 0>, rotation=<rot 90 0 0>}]
 dyn_collar_02 [{shapeType=Capsule, length=2, radius=1, shapeName=dyn_collar_02_capsule, center=<2 0 0>, rotation=<rot 90 0 0>}]
 dyn_skirt_b_l_01 [{shapeType=Capsule, length=10, radius=8, center=<6 0 8.3>, rotation=<rot 90 90 0>}]
 dyn_skirt_b_l_02 [{shapeType=Capsule, length=10, radius=8, center=<6 0 8>, rotation=<rot 90 90 0>}]
 dyn_skirt_b_l_03 [{shapeType=Capsule, length=10, radius=9, shapeName=dyn_skirt_f2_l_04_capsule, center=<6 0 8>, rotation=<rot 90 90 0>}]
 dyn_skirt_b_l_04 [{shapeType=Capsule, length=10, radius=9, shapeName=dyn_skirt_f2_l_04_capsule, center=<6 0 8>, rotation=<rot 90 90 0>}]
 dyn_skirt_b_l_base [{shapeType=Capsule, length=0.505, radius=0.505, shapeName=dyn_skirt_b_l_base_capsule, rotation=<rot 0 -90 0>}]
 dyn_skirt_b_r_01 [{shapeType=Capsule, length=10, radius=8, shapeName=dyn_skirt_b_r_01_capsule, center=<6 0 8.3>, rotation=<rot 90 90 0>}]
 dyn_skirt_b_r_02 [{shapeType=Capsule, length=10, radius=9, shapeName=dyn_skirt_b_r_02_capsule, center=<6 0 8>, rotation=<rot 90 90 0>}]
 dyn_skirt_b_r_03 [{shapeType=Capsule, length=10, radius=9, shapeName=dyn_skirt_f2_l_04_capsule, center=<6 0 8>, rotation=<rot 90 90 0>}]
 dyn_skirt_b_r_04 [{shapeType=Capsule, length=10, radius=9, shapeName=dyn_skirt_f2_l_04_capsule, center=<6 0 8>, rotation=<rot 90 90 0>}]
 dyn_skirt_b_r_base [{shapeType=Capsule, length=0.505, radius=0.505, shapeName=dyn_skirt_b_r_base_capsule, rotation=<rot 0 -90 0>}]
 dyn_skirt_f2_l_01 [{shapeType=Capsule, length=10, radius=8, shapeName=dyn_skirt_f2_l_01_capsule, center=<6 0 9>, rotation=<rot 90 90 0>}]
 dyn_skirt_f2_l_02 [{shapeType=Capsule, length=10, radius=8, shapeName=dyn_skirt_f2_l_02_capsule, center=<6 0 8>, rotation=<rot 90 90 0>}]
 dyn_skirt_f2_l_03 [{shapeType=Capsule, length=10, radius=9, shapeName=dyn_skirt_f2_l_04_capsule, center=<6 0 8>, rotation=<rot 90 90 0>}]
 dyn_skirt_f2_l_04 [{shapeType=Capsule, length=10, radius=9, shapeName=dyn_skirt_f2_l_04_capsule, center=<6 0 8>, rotation=<rot 90 90 0>}]
 dyn_skirt_f2_l_base [{shapeType=Capsule, length=0.505, radius=0.505, shapeName=dyn_skirt_f2_l_base_capsule, rotation=<rot 0 -90 0>}]
 dyn_skirt_f2_r_01 [{shapeType=Capsule, length=10, radius=8, shapeName=dyn_skirt_f2_r_01_capsule, center=<6 0 9>, rotation=<rot 90 90 0>}]
 dyn_skirt_f2_r_02 [{shapeType=Capsule, length=10, radius=8, shapeName=dyn_skirt_f2_r_02_capsule, center=<6 0 9>, rotation=<rot 90 90 0>}]
 dyn_skirt_f2_r_03 [{shapeType=Capsule, length=10, radius=9, shapeName=dyn_skirt_f2_l_04_capsule, center=<6 0 8>, rotation=<rot 90 90 0>}]
 dyn_skirt_f2_r_04 [{shapeType=Capsule, length=10, radius=9, shapeName=dyn_skirt_f2_l_04_capsule, center=<6 0 8>, rotation=<rot 90 90 0>}]
 dyn_skirt_f2_r_base [{shapeType=Capsule, length=0.505, radius=0.505, shapeName=dyn_skirt_f2_r_base_capsule, rotation=<rot 0 -90 0>}]
 dyn_skirt_f_c_01 [{shapeType=Capsule, length=10, radius=8, shapeName=dyn_skirt_f_c_01_capsule, center=<6 0 8>, rotation=<rot 90 90 0>}]
 dyn_skirt_f_c_02 [{shapeType=Capsule, length=10, radius=8, shapeName=dyn_skirt_f_c_02_capsule, center=<6 0 7>, rotation=<rot 90 90 0>}]
 dyn_skirt_f_c_03 [{shapeType=Capsule, length=10, radius=9, shapeName=dyn_skirt_f2_l_04_capsule, center=<6 0 7>, rotation=<rot 90 90 0>}]
 dyn_skirt_f_c_04 [{shapeType=Capsule, length=10, radius=9, shapeName=dyn_skirt_f2_l_04_capsule, center=<6 0 5.5>, rotation=<rot 90 90 0>}]
 dyn_skirt_f_c_base [{shapeType=Capsule, length=0.505, radius=0.505, shapeName=dyn_skirt_f_c_base_capsule, rotation=<rot 0 -90 0>}]
 dyn_skirt_f_l_01 [{shapeType=Capsule, length=10, radius=8, shapeName=dyn_skirt_f_l_01_capsule, center=<6 0 9>, rotation=<rot 90 90 0>}]
 dyn_skirt_f_l_02 [{shapeType=Capsule, length=10, radius=8, shapeName=dyn_skirt_f_l_02_capsule, center=<6 0 8>, rotation=<rot 90 90 0>}]
 dyn_skirt_f_l_03 [{shapeType=Capsule, length=10, radius=9, shapeName=dyn_skirt_f2_l_04_capsule, center=<6 0 8>, rotation=<rot 90 90 0>}]
 dyn_skirt_f_l_04 [{shapeType=Capsule, length=10, radius=9, shapeName=dyn_skirt_f2_l_04_capsule, center=<6 0 9>, rotation=<rot 90 90 0>}]
 dyn_skirt_f_l_base [{shapeType=Capsule, length=0.505, radius=0.505, shapeName=dyn_skirt_f_l_base_capsule, rotation=<rot 0 -90 0>}]
 dyn_skirt_f_r_01 [{shapeType=Capsule, length=10, radius=8, shapeName=dyn_skirt_f_r_01_capsule, center=<6 0 9>, rotation=<rot 90 90 0>}]
 dyn_skirt_f_r_02 [{shapeType=Capsule, length=10, radius=8, shapeName=dyn_skirt_f_r_02_capsule, center=<6 0 8>, rotation=<rot 90 90 0>}]
 dyn_skirt_f_r_03 [{shapeType=Capsule, length=10, radius=9, shapeName=dyn_skirt_f2_l_04_capsule, center=<6 0 8>, rotation=<rot 90 90 0>}]
 dyn_skirt_f_r_04 [{shapeType=Capsule, length=10, radius=9, shapeName=dyn_skirt_f2_l_04_capsule, center=<6 0 9>, rotation=<rot 90 90 0>}]
 dyn_skirt_f_r_base [{shapeType=Capsule, length=0.505, radius=0.505, shapeName=dyn_skirt_f_r_base_capsule, rotation=<rot 0 -90 0>}]
 dyn_skirt_s_l_01 [{shapeType=Capsule, length=10, radius=8, shapeName=dyn_skirt_s_l_01_capsule, center=<6 0 9>, rotation=<rot 90 90 0>}]
 dyn_skirt_s_l_02 [{shapeType=Capsule, length=10, radius=8, shapeName=dyn_skirt_s_l_02_capsule, center=<6 0 8>, rotation=<rot 90 90 0>}]
 dyn_skirt_s_l_03 [{shapeType=Capsule, length=10, radius=9, shapeName=dyn_skirt_f2_l_04_capsule, center=<6 0 7>, rotation=<rot 90 90 0>}]
 dyn_skirt_s_l_04 [{shapeType=Capsule, length=10, radius=9, shapeName=dyn_skirt_f2_l_04_capsule, center=<6 0 8>, rotation=<rot 90 90 0>}]
 dyn_skirt_s_l_base [{shapeType=Capsule, length=0.505, radius=0.505, shapeName=dyn_skirt_s_l_base_capsule, rotation=<rot 0 -90 0>}]
 dyn_skirt_s_r_01 [{shapeType=Capsule, length=10, radius=8, shapeName=dyn_skirt_s_r_01_capsule, center=<6 0 9>, rotation=<rot 90 90 0>}]
 dyn_skirt_s_r_02 [{shapeType=Capsule, length=10, radius=8, shapeName=dyn_skirt_s_r_02_capsule, center=<6 0 8>, rotation=<rot 90 90 0>}]
 dyn_skirt_s_r_03 [{shapeType=Capsule, length=10, radius=9, shapeName=dyn_skirt_f2_l_04_capsule, center=<6 0 7>, rotation=<rot 90 90 0>}]
 dyn_skirt_s_r_04 [{shapeType=Capsule, length=10, radius=9, shapeName=dyn_skirt_f2_l_04_capsule, center=<6 0 8>, rotation=<rot 90 90 0>}]
 dyn_skirt_s_r_base [{shapeType=Capsule, length=0.505, radius=0.505, shapeName=dyn_skirt_s_r_base_capsule, rotation=<rot 0 -90 0>}]
 foot_l [{shapeType=Capsule, length=12, radius=12, rotation=<rot 90 0 90>}]
 foot_r [{shapeType=Capsule, length=12, radius=12, rotation=<rot 90 0 90>}]
 pelvis [{shapeType=Capsule, length=9.958, radius=9.958, shapeName=pelvis_capsule, center=<-0.763 -1.9628 0>}]
 root [{shapeType=Capsule, length=0.505, radius=500, shapeName=root_sphere, center=<0 17.4691 -494.0154>, rotation=<rot 0 -90 0>}]
 spine_01 [{shapeType=Capsule, length=8.4251, radius=8.4251, shapeName=spine_01_capsule}]
 spine_02 [{shapeType=Capsule, length=7.0783, radius=7.0783, shapeName=spine_02_capsule}]
 spine_05 [
  {shapeType=Capsule, length=5, radius=14, shapeName=spine_05_capsule, center=<-6.4444 -0.0524 0>}
  {shapeType=Capsule, length=4.6413, radius=4.6413, shapeName=spine_04_capsule, center=<-4.8609 9.4003 2.1961>, rotation=<rot 30.3521 0 0>}
  {
   shapeType=Capsule
   length=5.3055
   radius=5.3055
   shapeName=spine_05_capsule
   center=<-5.113 8.929 -3.3791>
   rotation=<rot -38.7712 0 -3.1216>
  }
 ]
 thigh_l [{shapeType=Capsule, length=27.2619, radius=9.6619, shapeName=thigh_l_capsule, center=<-18.5908 -1.9108 1.7209>, rotation=<rot -88.2881 0 0>}]
 thigh_r [{shapeType=Capsule, length=27.2619, radius=9.6619, shapeName=thigh_r_capsule, center=<18.198 1.9108 -1.7327>, rotation=<rot -88.2881 0 0>}]
}
constraints [
 {
  bone1Name=pelvis
  bone2Name=root
  swing1LimitDegrees=45
  swing1Motion=Free
  swing2LimitDegrees=45
  swing2Motion=Free
  twistLimitDegrees=45
  twistMotion=Free
 }
 {
  bone1Name=base_beltfront
  bone2Name=pelvis
  swing1LimitDegrees=45
  swing1Motion=Free
  swing2LimitDegrees=45
  swing2Motion=Free
  twistLimitDegrees=45
  twistMotion=Free
 }
 {bone1Name=dyn_beltfront, bone2Name=base_beltfront, swing1LimitDegrees=20, swing2LimitDegrees=20, twistLimitDegrees=2}
 {
  bone1Name=base_beltfront_l
  bone2Name=pelvis
  swing1LimitDegrees=45
  swing1Motion=Free
  swing2LimitDegrees=45
  swing2Motion=Free
  twistLimitDegrees=45
  twistMotion=Free
 }
 {bone1Name=dyn_beltfront_l, bone2Name=base_beltfront_l, swing1LimitDegrees=20, swing2LimitDegrees=20, twistLimitDegrees=2}
 {
  bone1Name=base_beltfront_r
  bone2Name=pelvis
  swing1LimitDegrees=45
  swing1Motion=Free
  swing2LimitDegrees=45
  swing2Motion=Free
  twistLimitDegrees=45
  twistMotion=Free
 }
 {bone1Name=dyn_beltfront_r, bone2Name=base_beltfront_r, swing1LimitDegrees=20, swing2LimitDegrees=20, twistLimitDegrees=2}
 {
  bone1Name=dyn_skirt_f_c_base
  bone2Name=pelvis
  swing1LimitDegrees=45
  swing1Motion=Free
  swing2LimitDegrees=45
  swing2Motion=Free
  twistLimitDegrees=45
  twistMotion=Free
 }
 {bone1Name=dyn_skirt_f_c_01, bone2Name=dyn_skirt_f_c_base, swing1LimitDegrees=10, swing2LimitDegrees=10, twistLimitDegrees=2}
 {bone1Name=dyn_skirt_f_c_02, bone2Name=dyn_skirt_f_c_01, swing1LimitDegrees=10, swing2LimitDegrees=10, twistLimitDegrees=2}
 {bone1Name=dyn_skirt_f_c_03, bone2Name=dyn_skirt_f_c_02, swing1LimitDegrees=10, swing2LimitDegrees=60, twistLimitDegrees=5}
 {bone1Name=dyn_skirt_f_c_04, bone2Name=dyn_skirt_f_c_03, swing1LimitDegrees=10, swing2LimitDegrees=60, twistLimitDegrees=5}
 {
  bone1Name=spine_01
  bone2Name=root
  swing1LimitDegrees=45
  swing1Motion=Free
  swing2LimitDegrees=45
  swing2Motion=Free
  twistLimitDegrees=45
  twistMotion=Free
 }
 {
  bone1Name=spine_02
  bone2Name=root
  swing1LimitDegrees=45
  swing1Motion=Free
  swing2LimitDegrees=45
  swing2Motion=Free
  twistLimitDegrees=45
  twistMotion=Free
 }
 {
  bone1Name=spine_05
  bone2Name=root
  swing1LimitDegrees=45
  swing1Motion=Free
  swing2LimitDegrees=45
  swing2Motion=Free
  twistLimitDegrees=45
  twistMotion=Free
 }
 {
  bone1Name=base_collar
  bone2Name=spine_05
  swing1LimitDegrees=45
  swing1Motion=Free
  swing2LimitDegrees=45
  swing2Motion=Free
  twistLimitDegrees=45
  twistMotion=Free
 }
 {bone1Name=dyn_collar_01, bone2Name=base_collar, swing1LimitDegrees=15, swing2LimitDegrees=15, twistLimitDegrees=1}
 {bone1Name=dyn_collar_02, bone2Name=dyn_collar_01, swing1LimitDegrees=25, swing2LimitDegrees=25, twistLimitDegrees=1}
 {
  bone1Name=thigh_l
  bone2Name=root
  swing1LimitDegrees=45
  swing1Motion=Free
  swing2LimitDegrees=45
  swing2Motion=Free
  twistLimitDegrees=45
  twistMotion=Free
 }
 {
  bone1Name=calf_l
  bone2Name=root
  swing1LimitDegrees=45
  swing1Motion=Free
  swing2LimitDegrees=45
  swing2Motion=Free
  twistLimitDegrees=45
  twistMotion=Free
 }
 {
  bone1Name=foot_l
  bone2Name=calf_l
  swing1LimitDegrees=45
  swing1Motion=Free
  swing2LimitDegrees=45
  swing2Motion=Free
  twistLimitDegrees=45
  twistMotion=Free
 }
 {
  bone1Name=dyn_skirt_b_l_base
  bone2Name=pelvis
  swing1LimitDegrees=45
  swing1Motion=Free
  swing2LimitDegrees=45
  swing2Motion=Free
  twistLimitDegrees=45
  twistMotion=Free
 }
 {bone1Name=dyn_skirt_b_l_01, bone2Name=dyn_skirt_b_l_base, swing1LimitDegrees=10, swing2LimitDegrees=10, twistLimitDegrees=2}
 {bone1Name=dyn_skirt_b_l_02, bone2Name=dyn_skirt_b_l_01, swing1LimitDegrees=10, swing2LimitDegrees=60, twistLimitDegrees=2}
 {bone1Name=dyn_skirt_b_l_03, bone2Name=dyn_skirt_b_l_02, swing1LimitDegrees=10, swing2LimitDegrees=45, twistLimitDegrees=5}
 {bone1Name=dyn_skirt_b_l_04, bone2Name=dyn_skirt_b_l_03, swing1LimitDegrees=10, swing2LimitDegrees=45, twistLimitDegrees=5}
 {
  bone1Name=dyn_skirt_f2_l_base
  bone2Name=pelvis
  swing1LimitDegrees=45
  swing1Motion=Free
  swing2LimitDegrees=45
  swing2Motion=Free
  twistLimitDegrees=45
  twistMotion=Free
 }
 {bone1Name=dyn_skirt_f2_l_01, bone2Name=dyn_skirt_f2_l_base, swing1LimitDegrees=10, swing2LimitDegrees=10, twistLimitDegrees=2}
 {bone1Name=dyn_skirt_f2_l_02, bone2Name=dyn_skirt_f2_l_01, swing1LimitDegrees=10, swing2LimitDegrees=10, twistLimitDegrees=2}
 {bone1Name=dyn_skirt_f2_l_03, bone2Name=dyn_skirt_f2_l_02, swing1LimitDegrees=10, swing2LimitDegrees=60, twistLimitDegrees=5}
 {bone1Name=dyn_skirt_f2_l_04, bone2Name=dyn_skirt_f2_l_03, swing1LimitDegrees=10, swing2LimitDegrees=60, twistLimitDegrees=5}
 {
  bone1Name=dyn_skirt_f_l_base
  bone2Name=pelvis
  swing1LimitDegrees=45
  swing1Motion=Free
  swing2LimitDegrees=45
  swing2Motion=Free
  twistLimitDegrees=45
  twistMotion=Free
 }
 {bone1Name=dyn_skirt_f_l_01, bone2Name=dyn_skirt_f_l_base, swing1LimitDegrees=10, swing2LimitDegrees=10, twistLimitDegrees=2}
 {bone1Name=dyn_skirt_f_l_02, bone2Name=dyn_skirt_f_l_01, swing1LimitDegrees=10, swing2LimitDegrees=10, twistLimitDegrees=2}
 {bone1Name=dyn_skirt_f_l_03, bone2Name=dyn_skirt_f_l_02, swing1LimitDegrees=10, swing2LimitDegrees=60, twistLimitDegrees=5}
 {bone1Name=dyn_skirt_f_l_04, bone2Name=dyn_skirt_f_l_03, swing1LimitDegrees=10, swing2LimitDegrees=60, twistLimitDegrees=5}
 {
  bone1Name=dyn_skirt_s_l_base
  bone2Name=pelvis
  swing1LimitDegrees=45
  swing1Motion=Free
  swing2LimitDegrees=45
  swing2Motion=Free
  twistLimitDegrees=45
  twistMotion=Free
 }
 {bone1Name=dyn_skirt_s_l_01, bone2Name=dyn_skirt_s_l_base, swing1LimitDegrees=10, swing2LimitDegrees=10, twistLimitDegrees=2}
 {bone1Name=dyn_skirt_s_l_02, bone2Name=dyn_skirt_s_l_01, swing1LimitDegrees=10, swing2LimitDegrees=60, twistLimitDegrees=2}
 {bone1Name=dyn_skirt_s_l_03, bone2Name=dyn_skirt_s_l_02, swing1LimitDegrees=10, swing2LimitDegrees=45, twistLimitDegrees=5}
 {bone1Name=dyn_skirt_s_l_04, bone2Name=dyn_skirt_s_l_03, swing1LimitDegrees=10, swing2LimitDegrees=45, twistLimitDegrees=5}
 {
  bone1Name=thigh_r
  bone2Name=root
  swing1LimitDegrees=45
  swing1Motion=Free
  swing2LimitDegrees=45
  swing2Motion=Free
  twistLimitDegrees=45
  twistMotion=Free
 }
 {
  bone1Name=calf_r
  bone2Name=root
  swing1LimitDegrees=45
  swing1Motion=Free
  swing2LimitDegrees=45
  swing2Motion=Free
  twistLimitDegrees=45
  twistMotion=Free
 }
 {
  bone1Name=foot_r
  bone2Name=calf_r
  swing1LimitDegrees=45
  swing1Motion=Free
  swing2LimitDegrees=45
  swing2Motion=Free
  twistLimitDegrees=45
  twistMotion=Free
 }
 {
  bone1Name=dyn_skirt_b_r_base
  bone2Name=pelvis
  swing1LimitDegrees=45
  swing1Motion=Free
  swing2LimitDegrees=45
  swing2Motion=Free
  twistLimitDegrees=45
  twistMotion=Free
 }
 {bone1Name=dyn_skirt_b_r_01, bone2Name=dyn_skirt_b_r_base, swing1LimitDegrees=10, swing2LimitDegrees=10, twistLimitDegrees=2}
 {bone1Name=dyn_skirt_b_r_02, bone2Name=dyn_skirt_b_r_01, swing1LimitDegrees=10, swing2LimitDegrees=60, twistLimitDegrees=2}
 {bone1Name=dyn_skirt_b_r_03, bone2Name=dyn_skirt_b_r_02, swing1LimitDegrees=10, swing2LimitDegrees=45, twistLimitDegrees=5}
 {bone1Name=dyn_skirt_b_r_04, bone2Name=dyn_skirt_b_r_03, swing1LimitDegrees=10, swing2LimitDegrees=45, twistLimitDegrees=5}
 {
  bone1Name=dyn_skirt_f2_r_base
  bone2Name=pelvis
  swing1LimitDegrees=45
  swing1Motion=Free
  swing2LimitDegrees=45
  swing2Motion=Free
  twistLimitDegrees=45
  twistMotion=Free
 }
 {bone1Name=dyn_skirt_f2_r_01, bone2Name=dyn_skirt_f2_r_base, swing1LimitDegrees=10, swing2LimitDegrees=10, twistLimitDegrees=2}
 {bone1Name=dyn_skirt_f2_r_02, bone2Name=dyn_skirt_f2_r_01, swing1LimitDegrees=10, swing2LimitDegrees=10, twistLimitDegrees=2}
 {bone1Name=dyn_skirt_f2_r_03, bone2Name=dyn_skirt_f2_r_02, swing1LimitDegrees=10, swing2LimitDegrees=60, twistLimitDegrees=5}
 {bone1Name=dyn_skirt_f2_r_04, bone2Name=dyn_skirt_f2_r_03, swing1LimitDegrees=10, swing2LimitDegrees=60, twistLimitDegrees=5}
 {
  bone1Name=dyn_skirt_f_r_base
  bone2Name=pelvis
  swing1LimitDegrees=45
  swing1Motion=Free
  swing2LimitDegrees=45
  swing2Motion=Free
  twistLimitDegrees=45
  twistMotion=Free
 }
 {bone1Name=dyn_skirt_f_r_01, bone2Name=dyn_skirt_f_r_base, swing1LimitDegrees=10, swing2LimitDegrees=10, twistLimitDegrees=2}
 {bone1Name=dyn_skirt_f_r_02, bone2Name=dyn_skirt_f_r_01, swing1LimitDegrees=10, swing2LimitDegrees=10, twistLimitDegrees=2}
 {bone1Name=dyn_skirt_f_r_03, bone2Name=dyn_skirt_f_r_02, swing1LimitDegrees=10, swing2LimitDegrees=60, twistLimitDegrees=5}
 {bone1Name=dyn_skirt_f_r_04, bone2Name=dyn_skirt_f_r_03, swing1LimitDegrees=10, swing2LimitDegrees=60, twistLimitDegrees=5}
 {
  bone1Name=dyn_skirt_s_r_base
  bone2Name=pelvis
  swing1LimitDegrees=45
  swing1Motion=Free
  swing2LimitDegrees=45
  swing2Motion=Free
  twistLimitDegrees=45
  twistMotion=Free
 }
 {bone1Name=dyn_skirt_s_r_01, bone2Name=dyn_skirt_s_r_base, swing1LimitDegrees=10, swing2LimitDegrees=10, twistLimitDegrees=2}
 {bone1Name=dyn_skirt_s_r_02, bone2Name=dyn_skirt_s_r_01, swing1LimitDegrees=10, swing2LimitDegrees=60, twistLimitDegrees=2}
 {bone1Name=dyn_skirt_s_r_03, bone2Name=dyn_skirt_s_r_02, swing1LimitDegrees=10, swing2LimitDegrees=45, twistLimitDegrees=5}
 {bone1Name=dyn_skirt_s_r_04, bone2Name=dyn_skirt_s_r_03, swing1LimitDegrees=10, swing2LimitDegrees=45, twistLimitDegrees=5}
]
