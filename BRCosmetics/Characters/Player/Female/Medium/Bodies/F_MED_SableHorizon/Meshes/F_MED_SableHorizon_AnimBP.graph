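# Post-process style Anim Blueprint for the SableHorizon female body cosmetic, adding secondary motion to its jewels, collar and skirt.
# A copied parent pose runs through Control Rig and pelvis-space RigidBody into a cached pose.
# Two Anim Dynamics chains, differing in sim-space settings, blend on front-end state, then constraints bind skirt bones to legs and bone offsets adjust skirt and belt.
# The EventGraph derives movement, crouch, backbling, gravity and idle-emote variables.
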AnimBlueprint /BRCosmetics/Characters/Player/Female/Medium/Bodies/F_MED_SableHorizon/Meshes/F_MED_SableHorizon_AnimBP
target_skeleton=&/F_MED_SableHorizon_Skeleton
graphs [
 AnimGraph {nodes [
   "Transform (Modify) Bone - Bone: dyn_collar_02"!(ComponentPose, bAlphaBoolEnabled:bool=True, Translation="0,0,-0.4", Rotation="2,0,-3", Scale="1,1,1") -> (Pose) {
    alphaInputType=Bool
    rotationMode=BMM_Additive
    rotationSpace=BCS_BoneSpace
    translationMode=BMM_Additive
    translationSpace=BCS_BoneSpace
    alphaBoolBlend {blendInTime=0.2, blendOutTime=0.2}
    alphaScaleBiasClamp {scale=0.8}
    boneToModify {boneName=dyn_collar_02}
    rotation {pitch=2, roll=-3}
    translation {z=-0.4}
   }
   n1 = Anim Dynamics - Bone: base_jewel_torso_f_r!(ComponentPose=Local To Component(LocalPose=Use cached pose 'RBan_Cache'()), bAlphaBoolEnabled:bool=True, ExternalForce="0,0,0") -> (Pose) {
    alphaInputType=Bool
    angularDampingOverride=1
    angularSpringConstant=10000
    bAngularSpring=true
    bOverrideAngularDamping=true
    bUseGravityOverride=true
    gravityScale=0.35
    simulationSpace=RootRelative
    alphaBoolBlend {blendInTime=0.2, blendOption=Linear, blendOutTime=0.2}
    boundBone {boneName=base_jewel_torso_f_r}
    componentLinearAccScale=<1 1 1>
    componentLinearVelScale=<1 1 1>
    gravityOverride {z=-980}
    physicsBodyDefinitions [{
      boundBone {boneName=base_jewel_torso_f_r}
      boxExtents=<13 13 13>
      constraintSetup {angularLimitsMax=<5 15 2>, angularLimitsMin=<-5 -10 -2>, angularTarget=<1 0 0>}
      localJointOffset=<3 0 0>
     }]
    simSpaceSettings {maxAngularAcceleration=35, maxAngularVelocity=7, simSpaceAngularAlpha=0.4}
   }
   n2 = Anim Dynamics - Bone: dyn_jewel_torso_f_l!(ComponentPose=@n1, bAlphaBoolEnabled:bool=True, ExternalForce="0,0,0") -> (Pose) {
    alphaInputType=Bool
    angularDampingOverride=1
    angularSpringConstant=10000
    bAngularSpring=true
    bOverrideAngularDamping=true
    bUseGravityOverride=true
    gravityScale=0.35
    simulationSpace=RootRelative
    alphaBoolBlend {blendInTime=0.2, blendOption=Linear, blendOutTime=0.2}
    boundBone {boneName=dyn_jewel_torso_f_l}
    componentLinearAccScale=<1 1 1>
    componentLinearVelScale=<1 1 1>
    gravityOverride {z=-980}
    physicsBodyDefinitions [{
      boundBone {boneName=dyn_jewel_torso_f_l}
      boxExtents=<13 13 13>
      constraintSetup {angularLimitsMax=<5 15 2>, angularLimitsMin=<-5 -10 -2>, angularTarget=<1 0 0>}
      localJointOffset=<3 0 0>
     }]
    simSpaceSettings {maxAngularAcceleration=35, maxAngularVelocity=7, simSpaceAngularAlpha=0.4}
   }
   n3 = Anim Dynamics - Bone: dyn_jewel_torso_l!(ComponentPose=@n2, bAlphaBoolEnabled:bool=True, ExternalForce="0,0,0") -> (Pose) {
    alphaInputType=Bool
    angularDampingOverride=0.9999
    angularSpringConstant=500
    bAngularSpring=true
    bOverrideAngularDamping=true
    bUseGravityOverride=true
    gravityScale=0.35
    simulationSpace=RootRelative
    alphaBoolBlend {blendInTime=0.2, blendOption=Linear, blendOutTime=0.2}
    boundBone {boneName=dyn_jewel_torso_l}
    componentAppliedLinearAccClamp=<3000 3000 3000>
    componentLinearAccScale=<0.5 0.5 0.5>
    componentLinearVelScale=<0.5 0.5 0.5>
    gravityOverride {z=-980}
    physicsBodyDefinitions [{
      boundBone {boneName=dyn_jewel_torso_l}
      boxExtents=<13 13 13>
      constraintSetup {angularLimitsMax=<5 5 2>, angularLimitsMin=<-5 -40 -2>, angularTarget=<1 0 0>}
      localJointOffset=<3 0 0>
     }]
    simSpaceSettings {maxAngularAcceleration=35, maxAngularVelocity=7, simSpaceAngularAlpha=0.4}
   }
   n4 = Anim Dynamics - Bone: dyn_jewel_torso_r!(ComponentPose=@n3, bAlphaBoolEnabled:bool=True, ExternalForce="0,0,0") -> (Pose) {
    alphaInputType=Bool
    angularDampingOverride=0.9999
    angularSpringConstant=500
    bAngularSpring=true
    bOverrideAngularDamping=true
    bUseGravityOverride=true
    gravityScale=0.35
    simulationSpace=RootRelative
    alphaBoolBlend {blendInTime=0.2, blendOption=Linear, blendOutTime=0.2}
    boundBone {boneName=dyn_jewel_torso_r}
    componentAppliedLinearAccClamp=<3000 3000 3000>
    componentLinearAccScale=<0.5 0.5 0.5>
    componentLinearVelScale=<0.5 0.5 0.5>
    gravityOverride {z=-980}
    physicsBodyDefinitions [{
      boundBone {boneName=dyn_jewel_torso_r}
      boxExtents=<13 13 13>
      constraintSetup {angularLimitsMax=<5 5 2>, angularLimitsMin=<-5 -40 -2>, angularTarget=<1 0 0>}
      localJointOffset=<3 0 0>
     }]
    simSpaceSettings {maxAngularAcceleration=35, maxAngularVelocity=7, simSpaceAngularAlpha=0.4}
   }
   n5 = Anim Dynamics - Bone: dyn_jewel_torso_bk!(ComponentPose=@n4, bAlphaBoolEnabled:bool=True, ExternalForce="0,0,0") -> (Pose) {
    alphaInputType=Bool
    angularDampingOverride=0.9999
    angularSpringConstant=500
    bAngularSpring=true
    bOverrideAngularDamping=true
    bUseGravityOverride=true
    gravityScale=0.35
    simulationSpace=RootRelative
    alphaBoolBlend {blendInTime=0.2, blendOption=Linear, blendOutTime=0.2}
    boundBone {boneName=dyn_jewel_torso_bk}
    componentAppliedLinearAccClamp=<3000 3000 3000>
    componentLinearAccScale=<0.5 0.5 0.5>
    componentLinearVelScale=<0.5 0.5 0.5>
    gravityOverride {z=-980}
    physicsBodyDefinitions [{
      boundBone {boneName=dyn_jewel_torso_bk}
      boxExtents=<13 13 13>
      constraintSetup {angularLimitsMax=<5 0 2>, angularLimitsMin=<-5 -15 -2>, angularTarget=<1 0 0>}
      localJointOffset=<3 0 0>
     }]
    simSpaceSettings {maxAngularAcceleration=35, maxAngularVelocity=7, simSpaceAngularAlpha=0.4}
   }
   n6 = Anim Dynamics - Bone: dyn_jewel_arm_r_bk!(ComponentPose=@n5, bAlphaBoolEnabled:bool=True, ExternalForce="0,0,0") -> (Pose) {
    alphaInputType=Bool
    angularDampingOverride=1
    angularSpringConstant=10000
    bAngularSpring=true
    bOverrideAngularDamping=true
    bUseGravityOverride=true
    gravityScale=0.35
    simulationSpace=RootRelative
    alphaBoolBlend {blendInTime=0.2, blendOption=Linear, blendOutTime=0.2}
    boundBone {boneName=dyn_jewel_arm_r_bk}
    componentLinearAccScale=<1 1 1>
    componentLinearVelScale=<1 1 1>
    gravityOverride {z=-980}
    physicsBodyDefinitions [{
      boundBone {boneName=dyn_jewel_arm_r_bk}
      boxExtents=<13 13 13>
      constraintSetup {angularLimitsMax=<5 15 10>, angularLimitsMin=<-5 -10 -10>, angularTarget=<1 0 0>}
      localJointOffset=<3 0 0>
     }]
    simSpaceSettings {maxAngularAcceleration=35, maxAngularVelocity=7, simSpaceAngularAlpha=0.2}
   }
   n7 = Anim Dynamics - Bone: base_jewel_arm_r_btt!(ComponentPose=@n6, bAlphaBoolEnabled:bool=True, ExternalForce="0,0,0") -> (Pose) {
    alphaInputType=Bool
    angularDampingOverride=1
    angularSpringConstant=10000
    bAngularSpring=true
    bOverrideAngularDamping=true
    bUseGravityOverride=true
    gravityScale=0.35
    simulationSpace=RootRelative
    alphaBoolBlend {blendInTime=0.2, blendOption=Linear, blendOutTime=0.2}
    boundBone {boneName=base_jewel_arm_r_btt}
    componentLinearAccScale=<1 1 1>
    componentLinearVelScale=<1 1 1>
    gravityOverride {z=-980}
    physicsBodyDefinitions [{
      boundBone {boneName=base_jewel_arm_r_btt}
      boxExtents=<13 13 13>
      constraintSetup {angularLimitsMax=<5 15 10>, angularLimitsMin=<-5 -10 -10>, angularTarget=<1 0 0>}
      localJointOffset=<3 0 0>
     }]
    simSpaceSettings {maxAngularAcceleration=35, maxAngularVelocity=7, simSpaceAngularAlpha=0.2}
   }
   n8 = Anim Dynamics - Bone: dyn_jewel_arm_r_fr!(ComponentPose=@n7, bAlphaBoolEnabled:bool=True, ExternalForce="0,0,0") -> (Pose) {
    alphaInputType=Bool
    angularDampingOverride=1
    angularSpringConstant=10000
    bAngularSpring=true
    bOverrideAngularDamping=true
    bUseGravityOverride=true
    gravityScale=0.35
    simulationSpace=RootRelative
    alphaBoolBlend {blendInTime=0.2, blendOption=Linear, blendOutTime=0.2}
    boundBone {boneName=dyn_jewel_arm_r_fr}
    componentLinearAccScale=<1 1 1>
    componentLinearVelScale=<1 1 1>
    gravityOverride {z=-980}
    physicsBodyDefinitions [{
      boundBone {boneName=dyn_jewel_arm_r_fr}
      boxExtents=<13 13 13>
      constraintSetup {angularLimitsMax=<5 15 10>, angularLimitsMin=<-5 -10 -10>, angularTarget=<1 0 0>}
      localJointOffset=<3 0 0>
     }]
    simSpaceSettings {maxAngularAcceleration=35, maxAngularVelocity=7, simSpaceAngularAlpha=0.2}
   }
   n9 = Anim Dynamics - Bone: dyn_jewel_arm_r_top!(ComponentPose=@n8, bAlphaBoolEnabled:bool=True, ExternalForce="0,0,0") -> (Pose) {
    alphaInputType=Bool
    angularDampingOverride=1
    angularSpringConstant=10000
    bAngularSpring=true
    bOverrideAngularDamping=true
    bUseGravityOverride=true
    gravityScale=0.35
    simulationSpace=RootRelative
    alphaBoolBlend {blendInTime=0.2, blendOption=Linear, blendOutTime=0.2}
    boundBone {boneName=dyn_jewel_arm_r_top}
    componentLinearAccScale=<1 1 1>
    componentLinearVelScale=<1 1 1>
    gravityOverride {z=-980}
    physicsBodyDefinitions [{
      boundBone {boneName=dyn_jewel_arm_r_top}
      boxExtents=<13 13 13>
      constraintSetup {angularLimitsMax=<5 15 10>, angularLimitsMin=<-5 -10 -10>, angularTarget=<1 0 0>}
      localJointOffset=<3 0 0>
     }]
    simSpaceSettings {maxAngularAcceleration=35, maxAngularVelocity=7, simSpaceAngularAlpha=0.2}
   }
   n10 = Anim Dynamics - Bone: dyn_collar_02!(ComponentPose=@n9, bAlphaBoolEnabled:bool=True, ExternalForce="0,0,0") -> (Pose) {
    alphaInputType=Bool
    angularDampingOverride=0.9999
    angularSpringConstant=500
    bAngularSpring=true
    bOverrideAngularDamping=true
    bUseGravityOverride=true
    gravityScale=0.35
    simulationSpace=RootRelative
    alphaBoolBlend {blendInTime=0.2, blendOption=Linear, blendOutTime=0.2}
    boundBone {boneName=dyn_collar_02}
    componentAppliedLinearAccClamp=<3000 3000 3000>
    componentLinearAccScale=<0.5 0.5 0.5>
    componentLinearVelScale=<0.5 0.5 0.5>
    gravityOverride {z=-980}
    physicsBodyDefinitions [{
      boundBone {boneName=dyn_collar_02}
      boxExtents=<13 13 13>
      constraintSetup {angularLimitsMax=<5 0 15>, angularLimitsMin=<-5 -15 -15>, angularTarget=<1 0 0>}
      localJointOffset=<3 0 0>
     }]
    simSpaceSettings {maxAngularAcceleration=35, maxAngularVelocity=7, simSpaceAngularAlpha=0.4}
   }
   n11 = Anim Dynamics - Bone: dyn_collar_01!(ComponentPose=@n10, bAlphaBoolEnabled:bool=True, ExternalForce="0,0,0") -> (Pose) {
    alphaInputType=Bool
    angularDampingOverride=0.9999
    angularSpringConstant=500
    bAngularSpring=true
    bOverrideAngularDamping=true
    bUseGravityOverride=true
    gravityScale=0.35
    simulationSpace=RootRelative
    alphaBoolBlend {blendInTime=0.2, blendOption=Linear, blendOutTime=0.2}
    boundBone {boneName=dyn_collar_01}
    componentAppliedLinearAccClamp=<3000 3000 3000>
    componentLinearAccScale=<0.5 0.5 0.5>
    componentLinearVelScale=<0.5 0.5 0.5>
    gravityOverride {z=-980}
    physicsBodyDefinitions [{
      boundBone {boneName=dyn_collar_01}
      boxExtents=<13 13 13>
      constraintSetup {angularLimitsMax=<5 0 15>, angularLimitsMin=<-5 -15 -15>, angularTarget=<1 0 0>}
      localJointOffset=<3 0 0>
     }]
    simSpaceSettings {maxAngularAcceleration=35, maxAngularVelocity=7, simSpaceAngularAlpha=0.4}
   }
   n12 = Component To Local(ComponentPose=@n11)
   n13 = Anim Dynamics - Bone: base_jewel_torso_f_r(ComponentPose=Local To Component(LocalPose=Use cached pose 'RBan_Cache'()), bAlphaBoolEnabled=True, ExternalForce="0,0,0") {
    alphaInputType=Bool
    angularDampingOverride=1
    angularSpringConstant=10000
    bAngularSpring=true
    bOverrideAngularDamping=true
    bUseGravityOverride=true
    gravityScale=0.35
    simulationSpace=RootRelative
    alphaBoolBlend {blendInTime=0.2, blendOption=Linear, blendOutTime=0.2}
    boundBone {boneName=base_jewel_torso_f_r}
    componentLinearAccScale=<1 1 1>
    componentLinearVelScale=<1 1 1>
    gravityOverride {z=-980}
    physicsBodyDefinitions [{
      boundBone {boneName=base_jewel_torso_f_r}
      boxExtents=<13 13 13>
      constraintSetup {angularLimitsMax=<5 15 2>, angularLimitsMin=<-5 -10 -2>, angularTarget=<1 0 0>}
      localJointOffset=<3 0 0>
     }]
   }
   n14 = Anim Dynamics - Bone: dyn_jewel_torso_f_l(ComponentPose=@n13, bAlphaBoolEnabled=True, ExternalForce="0,0,0") {
    alphaInputType=Bool
    angularDampingOverride=1
    angularSpringConstant=10000
    bAngularSpring=true
    bOverrideAngularDamping=true
    bUseGravityOverride=true
    gravityScale=0.35
    simulationSpace=RootRelative
    alphaBoolBlend {blendInTime=0.2, blendOption=Linear, blendOutTime=0.2}
    boundBone {boneName=dyn_jewel_torso_f_l}
    componentLinearAccScale=<1 1 1>
    componentLinearVelScale=<1 1 1>
    gravityOverride {z=-980}
    physicsBodyDefinitions [{
      boundBone {boneName=dyn_jewel_torso_f_l}
      boxExtents=<13 13 13>
      constraintSetup {angularLimitsMax=<5 15 2>, angularLimitsMin=<-5 -10 -2>, angularTarget=<1 0 0>}
      localJointOffset=<3 0 0>
     }]
   }
   n15 = Anim Dynamics - Bone: dyn_jewel_torso_l(ComponentPose=@n14, bAlphaBoolEnabled=True, ExternalForce="0,0,0") {
    alphaInputType=Bool
    angularDampingOverride=0.9999
    angularSpringConstant=500
    bAngularSpring=true
    bOverrideAngularDamping=true
    bUseGravityOverride=true
    gravityScale=0.35
    simulationSpace=RootRelative
    alphaBoolBlend {blendInTime=0.2, blendOption=Linear, blendOutTime=0.2}
    boundBone {boneName=dyn_jewel_torso_l}
    componentAppliedLinearAccClamp=<3000 3000 3000>
    componentLinearAccScale=<0.5 0.5 0.5>
    componentLinearVelScale=<0.5 0.5 0.5>
    gravityOverride {z=-980}
    physicsBodyDefinitions [{
      boundBone {boneName=dyn_jewel_torso_l}
      boxExtents=<13 13 13>
      constraintSetup {angularLimitsMax=<5 5 2>, angularLimitsMin=<-5 -40 -2>, angularTarget=<1 0 0>}
      localJointOffset=<3 0 0>
     }]
   }
   n16 = Anim Dynamics - Bone: dyn_jewel_torso_r(ComponentPose=@n15, bAlphaBoolEnabled=True, ExternalForce="0,0,0") {
    alphaInputType=Bool
    angularDampingOverride=0.9999
    angularSpringConstant=500
    bAngularSpring=true
    bOverrideAngularDamping=true
    bUseGravityOverride=true
    gravityScale=0.35
    simulationSpace=RootRelative
    alphaBoolBlend {blendInTime=0.2, blendOption=Linear, blendOutTime=0.2}
    boundBone {boneName=dyn_jewel_torso_r}
    componentAppliedLinearAccClamp=<3000 3000 3000>
    componentLinearAccScale=<0.5 0.5 0.5>
    componentLinearVelScale=<0.5 0.5 0.5>
    gravityOverride {z=-980}
    physicsBodyDefinitions [{
      boundBone {boneName=dyn_jewel_torso_r}
      boxExtents=<13 13 13>
      constraintSetup {angularLimitsMax=<5 5 2>, angularLimitsMin=<-5 -40 -2>, angularTarget=<1 0 0>}
      localJointOffset=<3 0 0>
     }]
   }
   n17 = Anim Dynamics - Bone: dyn_jewel_torso_bk(ComponentPose=@n16, bAlphaBoolEnabled=True, ExternalForce="0,0,0") {
    alphaInputType=Bool
    angularDampingOverride=0.9999
    angularSpringConstant=500
    bAngularSpring=true
    bOverrideAngularDamping=true
    bUseGravityOverride=true
    gravityScale=0.35
    simulationSpace=RootRelative
    alphaBoolBlend {blendInTime=0.2, blendOption=Linear, blendOutTime=0.2}
    boundBone {boneName=dyn_jewel_torso_bk}
    componentAppliedLinearAccClamp=<3000 3000 3000>
    componentLinearAccScale=<0.5 0.5 0.5>
    componentLinearVelScale=<0.5 0.5 0.5>
    gravityOverride {z=-980}
    physicsBodyDefinitions [{
      boundBone {boneName=dyn_jewel_torso_bk}
      boxExtents=<13 13 13>
      constraintSetup {angularLimitsMax=<5 0 2>, angularLimitsMin=<-5 -15 -2>, angularTarget=<1 0 0>}
      localJointOffset=<3 0 0>
     }]
   }
   n18 = Anim Dynamics - Bone: dyn_jewel_arm_r_bk(ComponentPose=@n17, bAlphaBoolEnabled=True, ExternalForce="0,0,0") {
    alphaInputType=Bool
    angularDampingOverride=1
    angularSpringConstant=10000
    bAngularSpring=true
    bOverrideAngularDamping=true
    bUseGravityOverride=true
    gravityScale=0.35
    simulationSpace=RootRelative
    alphaBoolBlend {blendInTime=0.2, blendOption=Linear, blendOutTime=0.2}
    boundBone {boneName=dyn_jewel_arm_r_bk}
    componentLinearAccScale=<1 1 1>
    componentLinearVelScale=<1 1 1>
    gravityOverride {z=-980}
    physicsBodyDefinitions [{
      boundBone {boneName=dyn_jewel_arm_r_bk}
      boxExtents=<13 13 13>
      constraintSetup {angularLimitsMax=<5 15 10>, angularLimitsMin=<-5 -10 -10>, angularTarget=<1 0 0>}
      localJointOffset=<3 0 0>
     }]
   }
   n19 = Anim Dynamics - Bone: base_jewel_arm_r_btt(ComponentPose=@n18, bAlphaBoolEnabled=True, ExternalForce="0,0,0") {
    alphaInputType=Bool
    angularDampingOverride=1
    angularSpringConstant=10000
    bAngularSpring=true
    bOverrideAngularDamping=true
    bUseGravityOverride=true
    gravityScale=0.35
    simulationSpace=RootRelative
    alphaBoolBlend {blendInTime=0.2, blendOption=Linear, blendOutTime=0.2}
    boundBone {boneName=base_jewel_arm_r_btt}
    componentLinearAccScale=<1 1 1>
    componentLinearVelScale=<1 1 1>
    gravityOverride {z=-980}
    physicsBodyDefinitions [{
      boundBone {boneName=base_jewel_arm_r_btt}
      boxExtents=<13 13 13>
      constraintSetup {angularLimitsMax=<5 15 10>, angularLimitsMin=<-5 -10 -10>, angularTarget=<1 0 0>}
      localJointOffset=<3 0 0>
     }]
   }
   n20 = Anim Dynamics - Bone: dyn_jewel_arm_r_fr(ComponentPose=@n19, bAlphaBoolEnabled=True, ExternalForce="0,0,0") {
    alphaInputType=Bool
    angularDampingOverride=1
    angularSpringConstant=10000
    bAngularSpring=true
    bOverrideAngularDamping=true
    bUseGravityOverride=true
    gravityScale=0.35
    simulationSpace=RootRelative
    alphaBoolBlend {blendInTime=0.2, blendOption=Linear, blendOutTime=0.2}
    boundBone {boneName=dyn_jewel_arm_r_fr}
    componentLinearAccScale=<1 1 1>
    componentLinearVelScale=<1 1 1>
    gravityOverride {z=-980}
    physicsBodyDefinitions [{
      boundBone {boneName=dyn_jewel_arm_r_fr}
      boxExtents=<13 13 13>
      constraintSetup {angularLimitsMax=<5 15 10>, angularLimitsMin=<-5 -10 -10>, angularTarget=<1 0 0>}
      localJointOffset=<3 0 0>
     }]
   }
   n21 = Anim Dynamics - Bone: dyn_jewel_arm_r_top(ComponentPose=@n20, bAlphaBoolEnabled=True, ExternalForce="0,0,0") {
    alphaInputType=Bool
    angularDampingOverride=1
    angularSpringConstant=10000
    bAngularSpring=true
    bOverrideAngularDamping=true
    bUseGravityOverride=true
    gravityScale=0.35
    simulationSpace=RootRelative
    alphaBoolBlend {blendInTime=0.2, blendOption=Linear, blendOutTime=0.2}
    boundBone {boneName=dyn_jewel_arm_r_top}
    componentLinearAccScale=<1 1 1>
    componentLinearVelScale=<1 1 1>
    gravityOverride {z=-980}
    physicsBodyDefinitions [{
      boundBone {boneName=dyn_jewel_arm_r_top}
      boxExtents=<13 13 13>
      constraintSetup {angularLimitsMax=<5 15 10>, angularLimitsMin=<-5 -10 -10>, angularTarget=<1 0 0>}
      localJointOffset=<3 0 0>
     }]
   }
   n22 = Component To Local(ComponentPose=@n21)
   n23 = Get bIsInFrontEnd()
   n24 = "Constraint - dyn_collar_01 (P)"!(
    ComponentPose=Local To Component(LocalPose=Blend Poses by bool(BlendPose_0=@n12, BlendPose_1=@n22, bActiveValue=@n23))
    bAlphaBoolEnabled:bool=True
    ConstraintWeights_0=1
   ) -> (Pose) {
    alphaInputType=Bool
    alphaBoolBlend {blendInTime=0.2, blendOutTime=0.2}
    boneToModify {boneName=dyn_collar_01}
    constraintSetup [{targetBone {boneName=spine_05}}]
    constraintWeights [1]
   }
   n25 = "Constraint - dyn_skirt_f2_r_01 (P)"(ComponentPose=@n24, Alpha=1, ConstraintWeights_0=1) {
    alphaBoolBlend {blendInTime=0.2, blendOutTime=0.2}
    boneToModify {boneName=dyn_skirt_f2_r_01}
    constraintSetup [{targetBone {boneName=thigh_r}}]
    constraintWeights [1]
   }
   n26 = "Constraint - dyn_skirt_f_r_01 (P)"(ComponentPose=@n25, Alpha=1, ConstraintWeights_0=1) {
    alphaBoolBlend {blendInTime=0.2, blendOutTime=0.2}
    boneToModify {boneName=dyn_skirt_f_r_01}
    constraintSetup [{targetBone {boneName=thigh_r}}]
    constraintWeights [1]
   }
   n27 = "Constraint - dyn_skirt_s_r_01 (P)"(ComponentPose=@n26, Alpha=1, ConstraintWeights_0=1) {
    alphaBoolBlend {blendInTime=0.2, blendOutTime=0.2}
    boneToModify {boneName=dyn_skirt_s_r_01}
    constraintSetup [{targetBone {boneName=thigh_r}}]
    constraintWeights [1]
   }
   n28 = "Constraint - dyn_skirt_b_r_01 (P)"(ComponentPose=@n27, Alpha=1, ConstraintWeights_0=1) {
    alphaBoolBlend {blendInTime=0.2, blendOutTime=0.2}
    boneToModify {boneName=dyn_skirt_b_r_01}
    constraintSetup [{targetBone {boneName=thigh_r}}]
    constraintWeights [1]
   }
   n29 = "Constraint - dyn_skirt_f2_r_02 (P)"(ComponentPose=@n28, Alpha=1, ConstraintWeights_0=1) {
    alphaBoolBlend {blendInTime=0.2, blendOutTime=0.2}
    boneToModify {boneName=dyn_skirt_f2_r_02}
    constraintSetup [{targetBone {boneName=thigh_r}}]
    constraintWeights [1]
   }
   n30 = "Constraint - dyn_skirt_f_r_02 (P)"(ComponentPose=@n29, Alpha=1, ConstraintWeights_0=1) {
    alphaBoolBlend {blendInTime=0.2, blendOutTime=0.2}
    boneToModify {boneName=dyn_skirt_f_r_02}
    constraintSetup [{targetBone {boneName=thigh_r}}]
    constraintWeights [1]
   }
   n31 = "Constraint - dyn_skirt_s_r_02 (P)"(ComponentPose=@n30, Alpha=1, ConstraintWeights_0=1) {
    alphaBoolBlend {blendInTime=0.2, blendOutTime=0.2}
    boneToModify {boneName=dyn_skirt_s_r_02}
    constraintSetup [{targetBone {boneName=thigh_r}}]
    constraintWeights [1]
   }
   n32 = "Constraint - dyn_skirt_b_r_02 (P)"(ComponentPose=@n31, Alpha=1, ConstraintWeights_0=1) {
    alphaBoolBlend {blendInTime=0.2, blendOutTime=0.2}
    boneToModify {boneName=dyn_skirt_b_r_02}
    constraintSetup [{targetBone {boneName=thigh_r}}]
    constraintWeights [1]
   }
   n33 = "Constraint - dyn_skirt_f2_r_03 (P)"(ComponentPose=@n32, Alpha=1, ConstraintWeights_0=1) {
    alphaBoolBlend {blendInTime=0.2, blendOutTime=0.2}
    boneToModify {boneName=dyn_skirt_f2_r_03}
    constraintSetup [{targetBone {boneName=thigh_r}}]
    constraintWeights [1]
   }
   n34 = "Constraint - dyn_skirt_f_r_03 (P)"(ComponentPose=@n33, Alpha=1, ConstraintWeights_0=1) {
    alphaBoolBlend {blendInTime=0.2, blendOutTime=0.2}
    boneToModify {boneName=dyn_skirt_f_r_03}
    constraintSetup [{targetBone {boneName=thigh_r}}]
    constraintWeights [1]
   }
   n35 = "Constraint - dyn_skirt_s_r_03 (P)"(ComponentPose=@n34, Alpha=1, ConstraintWeights_0=1) {
    alphaBoolBlend {blendInTime=0.2, blendOutTime=0.2}
    boneToModify {boneName=dyn_skirt_s_r_03}
    constraintSetup [{targetBone {boneName=thigh_r}}]
    constraintWeights [1]
   }
   n36 = "Constraint - dyn_skirt_b_r_03 (P)"(ComponentPose=@n35, Alpha=1, ConstraintWeights_0=1) {
    alphaBoolBlend {blendInTime=0.2, blendOutTime=0.2}
    boneToModify {boneName=dyn_skirt_b_r_03}
    constraintSetup [{targetBone {boneName=calf_r}}]
    constraintWeights [1]
   }
   n37 = "Constraint - dyn_skirt_f2_r_04 (PP)"!(ComponentPose=@n36, Alpha=1, ConstraintWeights_0=0.6, ConstraintWeights_1:float=0.4) -> (Pose) {
    alphaBoolBlend {blendInTime=0.2, blendOutTime=0.2}
    boneToModify {boneName=dyn_skirt_f2_r_04}
    constraintSetup [{targetBone {boneName=thigh_r}}, {targetBone {boneName=calf_r}}]
    constraintWeights [1,1]
   }
   n38 = "Constraint - dyn_skirt_f_r_04 (PP)"!(ComponentPose=@n37, Alpha=1, ConstraintWeights_0=0.6, ConstraintWeights_1:float=0.4) -> (Pose) {
    alphaBoolBlend {blendInTime=0.2, blendOutTime=0.2}
    boneToModify {boneName=dyn_skirt_f_r_04}
    constraintSetup [{targetBone {boneName=thigh_r}}, {targetBone {boneName=calf_r}}]
    constraintWeights [1,1]
   }
   n39 = "Constraint - dyn_skirt_s_r_04 (PP)"!(ComponentPose=@n38, Alpha=1, ConstraintWeights_0=0.6, ConstraintWeights_1:float=0.4) -> (Pose) {
    alphaBoolBlend {blendInTime=0.2, blendOutTime=0.2}
    boneToModify {boneName=dyn_skirt_s_r_04}
    constraintSetup [{targetBone {boneName=thigh_r}}, {targetBone {boneName=calf_r}}]
    constraintWeights [1,1]
   }
   n40 = "Constraint - dyn_skirt_b_r_04 (PP)"!(ComponentPose=@n39, Alpha=1, ConstraintWeights_0=0, ConstraintWeights_1:float=1) -> (Pose) {
    alphaBoolBlend {blendInTime=0.2, blendOutTime=0.2}
    boneToModify {boneName=dyn_skirt_b_r_04}
    constraintSetup [{targetBone {boneName=thigh_r}}, {targetBone {boneName=calf_r}}]
    constraintWeights [1,1]
   }
   n41 = "Constraint - dyn_skirt_f2_l_01 (R)"(ComponentPose=@n40, Alpha=1, ConstraintWeights_0=1) {
    alphaBoolBlend {blendInTime=0.2, blendOutTime=0.2}
    boneToModify {boneName=dyn_skirt_f2_l_01}
    constraintSetup [{transformType=Rotation, targetBone {boneName=thigh_l}}]
    constraintWeights [1]
   }
   n42 = "Constraint - dyn_skirt_f_l_01 (R)"(ComponentPose=@n41, Alpha=1, ConstraintWeights_0=1) {
    alphaBoolBlend {blendInTime=0.2, blendOutTime=0.2}
    boneToModify {boneName=dyn_skirt_f_l_01}
    constraintSetup [{transformType=Rotation, targetBone {boneName=thigh_l}}]
    constraintWeights [1]
   }
   n43 = "Constraint - dyn_skirt_s_l_01 (R)"(ComponentPose=@n42, Alpha=1, ConstraintWeights_0=1) {
    alphaBoolBlend {blendInTime=0.2, blendOutTime=0.2}
    boneToModify {boneName=dyn_skirt_s_l_01}
    constraintSetup [{transformType=Rotation, targetBone {boneName=thigh_l}}]
    constraintWeights [1]
   }
   n44 = "Constraint - dyn_skirt_b_l_01 (R)"(ComponentPose=@n43, Alpha=1, ConstraintWeights_0=1) {
    alphaBoolBlend {blendInTime=0.2, blendOutTime=0.2}
    boneToModify {boneName=dyn_skirt_b_l_01}
    constraintSetup [{transformType=Rotation, targetBone {boneName=thigh_l}}]
    constraintWeights [1]
   }
   n45 = "Constraint - dyn_skirt_f2_l_02 (R)"(ComponentPose=@n44, Alpha=1, ConstraintWeights_0=1) {
    alphaBoolBlend {blendInTime=0.2, blendOutTime=0.2}
    boneToModify {boneName=dyn_skirt_f2_l_02}
    constraintSetup [{transformType=Rotation, targetBone {boneName=thigh_l}}]
    constraintWeights [1]
   }
   n46 = "Constraint - dyn_skirt_f_l_02 (R)"(ComponentPose=@n45, Alpha=1, ConstraintWeights_0=1) {
    alphaBoolBlend {blendInTime=0.2, blendOutTime=0.2}
    boneToModify {boneName=dyn_skirt_f_l_02}
    constraintSetup [{transformType=Rotation, targetBone {boneName=thigh_l}}]
    constraintWeights [1]
   }
   n47 = "Constraint - dyn_skirt_s_l_02 (R)"(ComponentPose=@n46, Alpha=1, ConstraintWeights_0=1) {
    alphaBoolBlend {blendInTime=0.2, blendOutTime=0.2}
    boneToModify {boneName=dyn_skirt_s_l_02}
    constraintSetup [{transformType=Rotation, targetBone {boneName=thigh_l}}]
    constraintWeights [1]
   }
   n48 = "Constraint - dyn_skirt_b_l_02 (R)"(ComponentPose=@n47, Alpha=1, ConstraintWeights_0=1) {
    alphaBoolBlend {blendInTime=0.2, blendOutTime=0.2}
    boneToModify {boneName=dyn_skirt_b_l_02}
    constraintSetup [{transformType=Rotation, targetBone {boneName=thigh_l}}]
    constraintWeights [1]
   }
   n49 = "Constraint - dyn_skirt_f2_l_03 (R)"(ComponentPose=@n48, Alpha=1, ConstraintWeights_0=1) {
    alphaBoolBlend {blendInTime=0.2, blendOutTime=0.2}
    boneToModify {boneName=dyn_skirt_f2_l_03}
    constraintSetup [{transformType=Rotation, targetBone {boneName=thigh_l}}]
    constraintWeights [1]
   }
   n50 = "Constraint - dyn_skirt_f_l_03 (R)"(ComponentPose=@n49, Alpha=1, ConstraintWeights_0=1) {
    alphaBoolBlend {blendInTime=0.2, blendOutTime=0.2}
    boneToModify {boneName=dyn_skirt_f_l_03}
    constraintSetup [{transformType=Rotation, targetBone {boneName=thigh_l}}]
    constraintWeights [1]
   }
   n51 = "Constraint - dyn_skirt_s_l_03 (R)"(ComponentPose=@n50, Alpha=1, ConstraintWeights_0=1) {
    alphaBoolBlend {blendInTime=0.2, blendOutTime=0.2}
    boneToModify {boneName=dyn_skirt_s_l_03}
    constraintSetup [{transformType=Rotation, targetBone {boneName=thigh_l}}]
    constraintWeights [1]
   }
   n52 = "Constraint - dyn_skirt_b_l_03 (R)"(ComponentPose=@n51, Alpha=1, ConstraintWeights_0=1) {
    alphaBoolBlend {blendInTime=0.2, blendOutTime=0.2}
    boneToModify {boneName=dyn_skirt_b_l_03}
    constraintSetup [{transformType=Rotation, targetBone {boneName=calf_l}}]
    constraintWeights [1]
   }
   n53 = "Constraint - dyn_skirt_f2_l_04 (PP)"!(ComponentPose=@n52, Alpha=1, ConstraintWeights_0=0.2, ConstraintWeights_1:float=0.8) -> (Pose) {
    alphaBoolBlend {blendInTime=0.2, blendOutTime=0.2}
    boneToModify {boneName=dyn_skirt_f2_l_04}
    constraintSetup [{targetBone {boneName=thigh_l}}, {targetBone {boneName=calf_l}}]
    constraintWeights [1,1]
   }
   n54 = "Constraint - dyn_skirt_f_l_04 (PP)"!(ComponentPose=@n53, Alpha=1, ConstraintWeights_0=0.2, ConstraintWeights_1:float=0.8) -> (Pose) {
    alphaBoolBlend {blendInTime=0.2, blendOutTime=0.2}
    boneToModify {boneName=dyn_skirt_f_l_04}
    constraintSetup [{targetBone {boneName=thigh_l}}, {targetBone {boneName=calf_l}}]
    constraintWeights [1,1]
   }
   n55 = "Constraint - dyn_skirt_s_l_04 (PP)"!(ComponentPose=@n54, Alpha=1, ConstraintWeights_0=0.2, ConstraintWeights_1:float=0.8) -> (Pose) {
    alphaBoolBlend {blendInTime=0.2, blendOutTime=0.2}
    boneToModify {boneName=dyn_skirt_s_l_04}
    constraintSetup [{targetBone {boneName=thigh_l}}, {targetBone {boneName=calf_l}}]
    constraintWeights [1,1]
   }
   n56 = "Constraint - dyn_skirt_b_l_04 (PP)"!(ComponentPose=@n55, Alpha=1, ConstraintWeights_0=0, ConstraintWeights_1:float=1) -> (Pose) {
    alphaBoolBlend {blendInTime=0.2, blendOutTime=0.2}
    boneToModify {boneName=dyn_skirt_b_l_04}
    constraintSetup [{targetBone {boneName=thigh_l}}, {targetBone {boneName=calf_l}}]
    constraintWeights [1,1]
   }
   n57 = "Constraint - dyn_skirt_f_c_01 (PP)"!(ComponentPose=@n56, Alpha=1, ConstraintWeights_0=0.5, ConstraintWeights_1:float=0.5) -> (Pose) {
    alphaBoolBlend {blendInTime=0.2, blendOutTime=0.2}
    boneToModify {boneName=dyn_skirt_f_c_01}
    constraintSetup [{targetBone {boneName=thigh_r}}, {targetBone {boneName=thigh_l}}]
    constraintWeights [1,1]
   }
   n58 = "Constraint - dyn_skirt_f_c_02 (PP)"!(ComponentPose=@n57, Alpha=1, ConstraintWeights_0=0.5, ConstraintWeights_1:float=0.5) -> (Pose) {
    alphaBoolBlend {blendInTime=0.2, blendOutTime=0.2}
    boneToModify {boneName=dyn_skirt_f_c_02}
    constraintSetup [{targetBone {boneName=thigh_r}}, {targetBone {boneName=thigh_l}}]
    constraintWeights [1,1]
   }
   n59 = "Constraint - dyn_skirt_f_c_03 (PP)"!(ComponentPose=@n58, Alpha=1, ConstraintWeights_0=0.5, ConstraintWeights_1:float=0.5) -> (Pose) {
    alphaBoolBlend {blendInTime=0.2, blendOutTime=0.2}
    boneToModify {boneName=dyn_skirt_f_c_03}
    constraintSetup [{targetBone {boneName=thigh_r}}, {targetBone {boneName=thigh_l}}]
    constraintWeights [1,1]
   }
   n60 = "Constraint - dyn_skirt_f_c_04 (RRRR)"!(
    ComponentPose=@n59
    Alpha=1
    ConstraintWeights_0=0.4
    ConstraintWeights_1:float=0.4
    ConstraintWeights_2:float=0.1
    ConstraintWeights_3:float=0.1
   ) -> (Pose) {
    alphaBoolBlend {blendInTime=0.2, blendOutTime=0.2}
    boneToModify {boneName=dyn_skirt_f_c_04}
    constraintSetup [
     {transformType=Rotation, targetBone {boneName=thigh_r}}
     {transformType=Rotation, targetBone {boneName=thigh_l}}
     {transformType=Rotation, targetBone {boneName=calf_l}}
     {transformType=Rotation, targetBone {boneName=calf_r}}
    ]
    constraintWeights [1,1,1,1]
   }
   n61 = "Transform (Modify) Bone - Bone: dyn_skirt_f2_r_base"(ComponentPose=@n60, Alpha=1, Translation="0,0,0", Rotation="15,0,0", Scale="1,1,1") {
    rotationMode=BMM_Additive
    rotationSpace=BCS_BoneSpace
    translationMode=BMM_Additive
    translationSpace=BCS_BoneSpace
    alphaBoolBlend {blendInTime=0.2, blendOutTime=0.2}
    boneToModify {boneName=dyn_skirt_f2_r_base}
    rotation {pitch=15}
   }
   n62 = "Transform (Modify) Bone - Bone: dyn_skirt_f2_l_base"(ComponentPose=@n61, Alpha=1, Translation="0,0,0", Rotation="15,0,0", Scale="1,1,1") {
    rotationMode=BMM_Additive
    rotationSpace=BCS_BoneSpace
    translationMode=BMM_Additive
    translationSpace=BCS_BoneSpace
    alphaBoolBlend {blendInTime=0.2, blendOutTime=0.2}
    boneToModify {boneName=dyn_skirt_f2_l_base}
    rotation {pitch=15}
   }
   n63 = "Transform (Modify) Bone - Bone: dyn_skirt_b_l_base"!(ComponentPose=@n62, bAlphaBoolEnabled:bool=True, Translation="0,0,0", Rotation="10,0,0", Scale="0.8,1,1") -> (Pose) {
    alphaInputType=Bool
    rotationMode=BMM_Additive
    rotationSpace=BCS_BoneSpace
    scaleMode=BMM_Additive
    scaleSpace=BCS_BoneSpace
    translationMode=BMM_Additive
    translationSpace=BCS_BoneSpace
    alphaBoolBlend {blendInTime=0.2, blendOutTime=0.2}
    boneToModify {boneName=dyn_skirt_b_l_base}
    rotation {pitch=10}
    scale {x=0.8}
   }
   n64 = "Transform (Modify) Bone - Bone: dyn_skirt_b_r_base"!(ComponentPose=@n63, bAlphaBoolEnabled:bool=True, Translation="0,0,0", Rotation="10,0,0", Scale="0.8,1,1") -> (Pose) {
    alphaInputType=Bool
    rotationMode=BMM_Additive
    rotationSpace=BCS_BoneSpace
    scaleMode=BMM_Additive
    scaleSpace=BCS_BoneSpace
    translationMode=BMM_Additive
    translationSpace=BCS_BoneSpace
    alphaBoolBlend {blendInTime=0.2, blendOutTime=0.2}
    alphaScaleBiasClamp {scale=0.8}
    boneToModify {boneName=dyn_skirt_b_r_base}
    rotation {pitch=10}
    scale {x=0.8}
   }
   n65 = "Transform (Modify) Bone - Bone: belt_front_main"!(ComponentPose=@n64, bAlphaBoolEnabled:bool=True, Translation="0,2,-1", Rotation="20,0,0", Scale="1,1,1") -> (Pose) {
    alphaInputType=Bool
    rotationMode=BMM_Additive
    rotationSpace=BCS_BoneSpace
    translationMode=BMM_Additive
    translationSpace=BCS_BoneSpace
    alphaBoolBlend {blendInTime=0.2, blendOutTime=0.2}
    alphaScaleBiasClamp {scale=0.8}
    boneToModify {boneName=belt_front_main}
    rotation {pitch=20}
    translation {y=2, z=-1}
   }
   n66 = "Transform (Modify) Bone - Bone: dyn_skirt_b_l_base"(ComponentPose=@n65, bAlphaBoolEnabled=True, Translation="0,0,0", Rotation="25,0,0", Scale="1,1,1") {
    alphaInputType=Bool
    rotationMode=BMM_Additive
    rotationSpace=BCS_BoneSpace
    scaleMode=BMM_Additive
    scaleSpace=BCS_BoneSpace
    translationMode=BMM_Additive
    translationSpace=BCS_BoneSpace
    alphaBoolBlend {blendInTime=0.2, blendOutTime=0.2}
    alphaScaleBiasClamp {scale=0.8}
    boneToModify {boneName=dyn_skirt_b_l_base}
    rotation {pitch=25}
   }
   n67 = "Transform (Modify) Bone - Bone: dyn_skirt_b_r_base"(ComponentPose=@n66, bAlphaBoolEnabled=True, Translation="0,0,0", Rotation="25,0,0", Scale="1,1,1") {
    alphaInputType=Bool
    rotationMode=BMM_Additive
    rotationSpace=BCS_BoneSpace
    scaleMode=BMM_Additive
    scaleSpace=BCS_BoneSpace
    translationMode=BMM_Additive
    translationSpace=BCS_BoneSpace
    alphaBoolBlend {blendInTime=0.2, blendOutTime=0.2}
    alphaScaleBiasClamp {scale=0.8}
    boneToModify {boneName=dyn_skirt_b_r_base}
    rotation {pitch=25}
   }
   Output Pose(Result=Component To Local(ComponentPose=@n67)) {layerGroup=None}
   n68 = Control Rig!(
    Source=Copy Pose From Mesh() {bUseAttachedParent=true}
    Alpha
    isSlidingVar_1:bool=Get bIsBaseballSliding()
    sideToside:double=Get sideToSide() -> (sideToSide:double)
    Is Crouching_0:bool=Get isCrouchingNotSliding() -> (isCrouchingNotSliding:bool)
    isEmote:bool=Get bIsPlayingEmote()
    isSlidingVar_2:bool=Get bIsKneeSliding()
    pawnSpeed:double=Get PawnSpeedAlpha() -> (PawnSpeedAlpha:double)
    crouchFloat:double=Get crouchFloat() -> (crouchFloat:double)
    isCapeWearing:bool=Get Is Wearing Cape() -> (Is Wearing Cape:bool)
   ) -> (Pose) {controlRigAssetReference {blueprintRigClass=&/F_MED_SableHorizon_CtrlRig.F_MED_SableHorizon_CtrlRig_C}}
   n69 = RigidBody!(ComponentPose=Local To Component(LocalPose=@n68), Alpha, OverrideWorldGravity:Vector=Get Gravity Override(), ExternalForce) -> (Pose) {
    bClampLinearTranslationLimitToRefPose=true
    bOverrideWorldGravity=true
    overridePhysicsAsset=&/F_MED_SableHorizon_Physics
    simulationSpace=BaseBoneSpace
    alphaBoolBlend {blendOption=Linear}
    baseBoneRef {boneName=pelvis}
    overrideWorldGravity {z=-980}
    simSpaceSettings {dampingAlpha=0.25, maxAngularAcceleration=8, maxAngularVelocity=2.5, maxLinearAcceleration=800, maxLinearVelocity=300, worldAlpha=1}
   }
   Save cached pose 'RBan_Cache'(Pose=Component To Local(ComponentPose=@n69)) {cacheName=RBan_Cache}
   Get is Knee Sliding() -> (is Knee Sliding:double)
  ]}
 EventGraph {nodes [
   n1 = Event BlueprintUpdateAnimation()
   > Set deltaTime(deltaTime:double=@n1.DeltaTimeX) -> (Output_Get:double)
   > n2 = Sequence!(execute) -> (then_0, then_1, then_2:Exec, then_3:Exec)
   > n3 = "Gravity Override State Function (Parameter Asset) (Nativized)"(
    CharacterPartAnimInstance=Self-Reference()
    GameTimeInSeconds=GetGameTimeInSeconds()
    Parameters=&/F_MED_SableHorizon_Parameters
   )
   > Set Gravity Override(Gravity Override=@n3)
   > Set RBAN Settings Out(RBAN Settings Out=@n3.OutRigidBodyAnimNodeParameters)
   > Set is Knee Sliding(is Knee Sliding:double="To Float (Boolean)"(InBool=Get bIsKneeSliding())) -> (Output_Get:double)
   n4 = Get bIsCrouching()
   > Set crouchFloat(crouchFloat:double=FInterpTo(Current=Get crouchFloat(), Target="To Float (Boolean)"(InBool=@n4), DeltaTime=@n1.DeltaTimeX, InterpSpeed=10)) -> (Output_Get:double)
   n5 = Get bIsSliding()
   > Set isCrouchingNotSliding(isCrouchingNotSliding:bool=AND Boolean(A=NOT Boolean(A=@n5), B=Get bIsCrouching())) -> (Output_Get:bool)
   n6 = Player Movement Function DEPRECATED!(
    execute=@n2.then_1
    CharacterPartAnimInstance:Custom Character Part Anim Instance Object Reference=Self-Reference()
    X Dot Product:double=100
    Y Dot Product:double=100
    Z Dot Product:double=100
    DrawDebug:bool=false
    Debug Scalar:double=50
   ) -> (then, X Movement:double, Y Movement:double, Z Movement:double)
   n7 = BreakVector(InVec=MakeVector(X=@n6, Y=@n6.Y Movement))
   > n8 = Set Left(Left:double="To Float (Boolean)"(InBool="float < float"(A=@n7, B=-50))) -> (Output_Get:double)
   > n9 = Set Right(Right:double="To Float (Boolean)"(InBool="float > float"(A=@n7, B=50))) -> (Output_Get:double)
   > Set Forward(Forward:double="To Float (Boolean)"(InBool="float > float"(A=@n7.Y, B=40))) -> (Output_Get:double)
   > Set Backward(Backward:double="To Float (Boolean)"(InBool="float < float"(A=@n7.Y, B=-40))) -> (Output_Get:double)
   > Set sideToSide(sideToSide:double=float + float(A=@n8, B=@n9)) -> (Output_Get:double)
   n10 = Set armFold(
    execute=@n2.then_3
    armFold:double=GetAlphaFromReferenceTransformDelta(SkeletalMeshComponent=Get Skel Mesh(), BoneName=upperarm_l, BaseName=lowerarm_l, InputRange=(X=-90,Y=0), TwistAxis="1,0,0")
   ) -> (Output_Get:double)
   > n11 = Compare Float!(Exec:Exec, Input:double=@n10, CompareWith:double=0.6) -> (">":Exec, "==":Exec, "<":Exec) {is=title K2Node_MacroInstance}
   n12 = AND Boolean(A=Get bIsInFrontEnd(), B=NOT Boolean(A=Get bIsPlayingEmote()))
   Set Is Custom Emote Playing_Idle2(execute=@n11."<", Is Custom Emote Playing_Idle2:bool=@n12) -> (Output_Get:bool)
   Set Is Custom Emote Playing_Idle1(execute=@n11.">", Is Custom Emote Playing_Idle1:bool=@n12) -> (Output_Get:bool)
   n13 = CheckForBackbling()
   Set isWearingBackbling(
    execute=@n2.then_2
    isWearingBackbling=AND Boolean!(A=@n13, B=NOT Boolean(A=@n13.bIsBackblingACape), C:bool=NOT Boolean(A=@n13.bIsBackblingLong)) -> (ReturnValue)
   )
   > Set Is Wearing Cape(Is Wearing Cape=OR Boolean(A=@n13.bIsBackblingACape, B=@n13.bIsBackblingLong))
   n14 = "To Float (Boolean)"(InBool=OR Boolean(A=Get bIsPlayingEmote(), B=Get bIsOnGrindRail()))
   n15 = Get deltaTime() -> (deltaTime:double)
   n16 = Get isSkirtCloser() -> (isSkirtCloser:double)
   > Set isSkirtCloser(isSkirtCloser:double="Clamp (Float)"(Value=FInterpTo(Current=@n16, Target=@n14, DeltaTime=@n15, InterpSpeed=10), Min=0, Max=0.7)) -> (Output_Get:double)
   Event BlueprintInitializeAnimation()
   > Set Skel Mesh(Skel Mesh=GetOwnerSkeletalMeshComponent())
   Event OnCharacterCustomizationLoaded()
   > n17 = Branch(Condition=Get bIsInFrontEnd())
   > Set Skirt Sim Space(Skirt Sim Space:AnimPhysSimSpaceType Enum=World) -> (Output_Get:AnimPhysSimSpaceType Enum)
   > Set Skirt Alpha(Skirt Alpha:float=0.1) -> (Output_Get:float)
   Set Skirt Sim Space(execute=@n17.else, Skirt Sim Space=BoneRelative)
   > Set Skirt Alpha(Skirt Alpha=1)
   Get bIsTacticalSprinting()
  ]}
]
variables [
 armFold {type=number}
 Backward {type=number}
 crouchFloat {type=number}
 deltaTime {type=number}
 Forward {type=number}
 Gravity Override {type=Vector}
 Is Custom Emote Playing_Idle1 {type=boolean}
 Is Custom Emote Playing_Idle2 {type=boolean}
 is Knee Sliding {type=number}
 Is Wearing Cape {type=boolean}
 isCrouchingNotSliding {type=boolean}
 isSkirtCloser {type=number}
 isWearingBackbling {type=boolean}
 Left {type=number}
 PawnSpeedAlpha {type=number}
 RBAN Settings Out {type=FortRigidBodyAnimNodeParameters}
 Right {type=number}
 sideToSide {type=number}
 Skel Mesh {type=SkeletalMeshComponent}
 Skirt Alpha {type=number}
 Skirt Sim Space {type=AnimPhysSimSpaceType}
]
