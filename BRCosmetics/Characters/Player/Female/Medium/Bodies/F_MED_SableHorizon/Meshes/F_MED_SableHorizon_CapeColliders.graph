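# Physics asset of collision volumes for the cape on the Sable Horizon medium female player body, per its name.
# Capsules wrap the head, clavicles, spine and back skirt bones, with extra capsules extending from the pelvis.
# A huge root sphere, offset downward so only its top rises above the origin, forms a floor-like collider.
# Limb and neck bodies carry no shapes, and the sole constraint, right hand to forearm, has free angular motion.
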PhysicsAsset /BRCosmetics/Characters/Player/Female/Medium/Bodies/F_MED_SableHorizon/Meshes/F_MED_SableHorizon_CapeColliders
bodies {
 calf_l []
 calf_r []
 clavicle_l [{shapeType=Capsule, length=6.5623, radius=7.5194, center=<3.9566 -1.0862 -2.1127>, rotation=<rot -77.5566 177.3915 7.7314>}]
 clavicle_r [{shapeType=Capsule, length=6.5623, radius=7.5194, center=<-3.9566 1.0862 2.1127>, rotation=<rot -77.5566 177.3915 -172.2686>}]
 dyn_skirt_b_l_02 [{
   shapeType=Capsule
   length=10.5018
   radius=8.1986
   shapeName=dyn_skirt_b_r_03_capsule
   center=<7.999 -1.5521 -5.2484>
   rotation=<rot 0.0239 6.1301 -103.168>
  }]
 dyn_skirt_b_l_03 [{
   shapeType=Capsule
   length=10.5018
   radius=8.1986
   shapeName=dyn_skirt_b_r_03_capsule
   center=<7.999 -1.5521 -5.2484>
   rotation=<rot 0.0239 6.1301 -103.168>
  }]
 dyn_skirt_b_r_02 [{
   shapeType=Capsule
   length=10.5018
   radius=8.1986
   shapeName=dyn_skirt_b_r_03_capsule
   center=<7.999 -1.5521 -5.2484>
   rotation=<rot 0.024 6.1301 -89.7473>
  }]
 dyn_skirt_b_r_03 [{
   shapeType=Capsule
   length=10.5018
   radius=8.1986
   shapeName=dyn_skirt_b_r_03_capsule
   center=<7.999 -1.5521 -5.2484>
   rotation=<rot 0.024 6.1301 -89.7473>
  }]
 foot_l []
 foot_r []
 hand_r []
 head [{shapeType=Capsule, length=0, radius=18, center=<9.0698 0.0746 0>, rotation=<rot 0 -90 0>}]
 lowerarm_l []
 lowerarm_r []
 neck_02 []
 pelvis [
  {shapeType=Capsule, length=8.6456, radius=14.5388, center=<-2.4887 0.1881 0>}
  {shapeType=Capsule, length=33, radius=15, shapeName=pelvis_capsule, center=<-20.4939 -3.3293 0>}
  {shapeType=Capsule, length=36, radius=15, shapeName=pelvis_capsule, center=<-45.805 -7.3311 0>}
 ]
 root [{shapeType=Sphere, radius=157, center=<0 0 -142.027>}]
 spine_01 [{shapeType=Capsule, length=23.4438, radius=13.5586, center=<-1.7965 -1.6951 0>}]
 spine_02 [
  {shapeType=Capsule, length=2.5219, radius=9.7317, center=<-2.6885 -1.2641 0>}
  {shapeType=Capsule, length=20.6875, radius=9.3125, center=<-1.8317 1.9362 0>, rotation=<rot 90 0 0>}
 ]
 spine_03 [
  {shapeType=Capsule, length=11, radius=9, center=<0.0574 0.4416 0>}
  {shapeType=Capsule, length=20.7188, radius=9.2812, center=<-1.2163 0.549 0>, rotation=<rot 90 180 -175.0732>}
 ]
 spine_04 [
  {shapeType=Capsule, length=11, radius=10.3916, center=<5.6234 -0.6779 0>}
  {shapeType=Capsule, length=10.0312, radius=9.2812, center=<-1.2326 1.308 0>, rotation=<rot 90 0 0>}
 ]
 spine_05 [{shapeType=Capsule, length=9.678, radius=15.3571, center=<0.6969 1.91 0>}]
 thigh_l []
 thigh_r []
 upperarm_l []
 upperarm_r []
}
constraints [{
  bone1Name=hand_r
  bone2Name=lowerarm_r
  swing1LimitDegrees=45
  swing1Motion=Free
  swing2LimitDegrees=45
  swing2Motion=Free
  twistLimitDegrees=45
  twistMotion=Free
 }]
