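# Post-copy AnimBP for the SableHorizon female face accessory, adding secondary motion to its dangling front and back chain bones.
# It copies the parent pose, layers lower-face bones from the pawn's Head part mesh, offsets jaw and front chain bases, then caches the result.
# Two six-node AnimDynamics variants with differing simulation-space settings are selected by bIsInFrontEnd.
# Low-alpha constraints then tie the front side chains to the head.
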AnimBlueprint /BRCosmetics/Characters/Player/Female/Medium/Bodies/F_MED_SableHorizon/Meshes/Parts/F_MED_SableHorizon_FaceAcc_AnimBP
target_skeleton=&/F_MED_SableHorizon_FaceAcc_Skeleton
graphs [
 AnimGraph {nodes [
   n1 = Anim Dynamics - Bone: dyn_chain_fr!(ComponentPose=Local To Component(LocalPose=Use cached pose 'falloff_cache'()), bAlphaBoolEnabled:bool=True, ExternalForce="0,0,0") -> (Pose) {
    alphaInputType=Bool
    angularDampingOverride=1
    angularSpringConstant=10000
    bAngularSpring=true
    bOverrideAngularDamping=true
    bUseGravityOverride=true
    gravityScale=0.35
    simulationSpace=Actor
    alphaBoolBlend {blendInTime=0.2, blendOption=Linear, blendOutTime=0.2}
    boundBone {boneName=dyn_chain_fr}
    componentLinearAccScale=<1 1 1>
    componentLinearVelScale=<1 1 1>
    gravityOverride {z=-980}
    physicsBodyDefinitions [{
      boundBone {boneName=dyn_chain_fr}
      boxExtents=<20 20 20>
      constraintSetup {angularLimitsMax=<2 0 10>, angularLimitsMin=<-2 -10 -10>, angularTarget=<1 0 0>}
      localJointOffset=<3 0 0>
     }]
    relativeSpaceBone {boneName=neck_02}
   }
   n2 = Anim Dynamics - Bone: dyn_chain_fr_l!(ComponentPose=@n1, bAlphaBoolEnabled:bool=True, ExternalForce="0,0,0") -> (Pose) {
    alphaInputType=Bool
    angularDampingOverride=1
    angularSpringConstant=10000
    bAngularSpring=true
    bOverrideAngularDamping=true
    bUseGravityOverride=true
    gravityScale=0.35
    simulationSpace=Actor
    alphaBoolBlend {blendInTime=0.2, blendOption=Linear, blendOutTime=0.2}
    boundBone {boneName=dyn_chain_fr_l}
    componentLinearAccScale=<1 1 1>
    componentLinearVelScale=<1 1 1>
    gravityOverride {z=-980}
    physicsBodyDefinitions [{
      boundBone {boneName=dyn_chain_fr_l}
      boxExtents=<20 20 20>
      constraintSetup {angularLimitsMax=<5 -5 5>, angularLimitsMin=<-5 -25 -5>, angularTarget=<1 0 0>}
      localJointOffset=<3 0 0>
     }]
    relativeSpaceBone {boneName=head}
   }
   n3 = Anim Dynamics - Bone: dyn_chain_fr_r!(ComponentPose=@n2, bAlphaBoolEnabled:bool=True, ExternalForce="0,0,0") -> (Pose) {
    alphaInputType=Bool
    angularDampingOverride=1
    angularSpringConstant=10000
    bAngularSpring=true
    bOverrideAngularDamping=true
    bUseGravityOverride=true
    gravityScale=0.35
    simulationSpace=Actor
    alphaBoolBlend {blendInTime=0.2, blendOption=Linear, blendOutTime=0.2}
    boundBone {boneName=dyn_chain_fr_r}
    componentLinearAccScale=<1 1 1>
    componentLinearVelScale=<1 1 1>
    gravityOverride {z=-980}
    physicsBodyDefinitions [{
      boundBone {boneName=dyn_chain_fr_r}
      boxExtents=<20 20 20>
      constraintSetup {angularLimitsMax=<5 -5 5>, angularLimitsMin=<-5 -25 -5>, angularTarget=<1 0 0>}
      localJointOffset=<3 0 0>
     }]
    relativeSpaceBone {boneName=head}
   }
   n4 = Anim Dynamics - Bone: dyn_chain_bk!(ComponentPose=@n3, bAlphaBoolEnabled:bool=True, ExternalForce="0,0,0") -> (Pose) {
    alphaInputType=Bool
    angularDampingOverride=1
    angularSpringConstant=10000
    bAngularSpring=true
    bOverrideAngularDamping=true
    bUseGravityOverride=true
    gravityScale=0.35
    simulationSpace=RootRelative
    alphaBoolBlend {blendInTime=0.2, blendOption=Linear, blendOutTime=0.2}
    boundBone {boneName=dyn_chain_bk}
    componentLinearAccScale=<1 1 1>
    componentLinearVelScale=<1 1 1>
    gravityOverride {z=-980}
    physicsBodyDefinitions [{
      boundBone {boneName=dyn_chain_bk}
      boxExtents=<20 20 20>
      constraintSetup {angularLimitsMax=<10 0 90>, angularLimitsMin=<-10 -90 -90>, angularTarget=<1 0 0>}
      localJointOffset=<3 0 0>
     }]
   }
   n5 = Anim Dynamics - Bone: dyn_chain_bk_l!(ComponentPose=@n4, bAlphaBoolEnabled:bool=True, ExternalForce="0,0,0") -> (Pose) {
    alphaInputType=Bool
    angularDampingOverride=1
    angularSpringConstant=10000
    bAngularSpring=true
    bOverrideAngularDamping=true
    bUseGravityOverride=true
    gravityScale=0.35
    simulationSpace=Actor
    alphaBoolBlend {blendInTime=0.2, blendOption=Linear, blendOutTime=0.2}
    boundBone {boneName=dyn_chain_bk_l}
    componentLinearAccScale=<1 1 1>
    componentLinearVelScale=<1 1 1>
    gravityOverride {z=-980}
    physicsBodyDefinitions [{
      boundBone {boneName=dyn_chain_bk_l}
      boxExtents=<20 20 20>
      constraintSetup {angularLimitsMax=<10 0 10>, angularLimitsMin=<-10 -25 -10>, angularTarget=<1 0 0>}
      localJointOffset=<3 0 0>
     }]
    relativeSpaceBone {boneName=head}
   }
   n6 = Anim Dynamics - Bone: dyn_chain_bk_r!(ComponentPose=@n5, bAlphaBoolEnabled:bool=True, ExternalForce="0,0,0") -> (Pose) {
    alphaInputType=Bool
    angularDampingOverride=1
    angularSpringConstant=10000
    bAngularSpring=true
    bOverrideAngularDamping=true
    bUseGravityOverride=true
    gravityScale=0.35
    simulationSpace=Actor
    alphaBoolBlend {blendInTime=0.2, blendOption=Linear, blendOutTime=0.2}
    boundBone {boneName=dyn_chain_bk_r}
    componentLinearAccScale=<1 1 1>
    componentLinearVelScale=<1 1 1>
    gravityOverride {z=-980}
    physicsBodyDefinitions [{
      boundBone {boneName=dyn_chain_bk_r}
      boxExtents=<20 20 20>
      constraintSetup {angularLimitsMax=<10 0 10>, angularLimitsMin=<-10 -25 -10>, angularTarget=<1 0 0>}
      localJointOffset=<3 0 0>
     }]
    relativeSpaceBone {boneName=head}
   }
   n7 = Component To Local(ComponentPose=@n6)
   n8 = Anim Dynamics - Bone: dyn_chain_fr(ComponentPose=Local To Component(LocalPose=Use cached pose 'falloff_cache'()), bAlphaBoolEnabled=True, ExternalForce="0,0,0") {
    alphaInputType=Bool
    angularDampingOverride=1
    angularSpringConstant=10000
    bAngularSpring=true
    bOverrideAngularDamping=true
    bUseGravityOverride=true
    gravityScale=0.35
    simulationSpace=Actor
    alphaBoolBlend {blendInTime=0.2, blendOption=Linear, blendOutTime=0.2}
    boundBone {boneName=dyn_chain_fr}
    componentLinearAccScale=<1 1 1>
    componentLinearVelScale=<1 1 1>
    gravityOverride {z=-980}
    physicsBodyDefinitions [{
      boundBone {boneName=dyn_chain_fr}
      boxExtents=<20 20 20>
      constraintSetup {angularLimitsMax=<2 0 10>, angularLimitsMin=<-2 -10 -10>, angularTarget=<1 0 0>}
      localJointOffset=<3 0 0>
     }]
    relativeSpaceBone {boneName=neck_02}
    simSpaceSettings {maxAngularAcceleration=35, maxAngularVelocity=7, simSpaceAngularAlpha=0.4}
   }
   n9 = Anim Dynamics - Bone: dyn_chain_fr_l(ComponentPose=@n8, bAlphaBoolEnabled=True, ExternalForce="0,0,0") {
    alphaInputType=Bool
    angularDampingOverride=1
    angularSpringConstant=10000
    bAngularSpring=true
    bOverrideAngularDamping=true
    bUseGravityOverride=true
    gravityScale=0.35
    simulationSpace=Actor
    alphaBoolBlend {blendInTime=0.2, blendOption=Linear, blendOutTime=0.2}
    boundBone {boneName=dyn_chain_fr_l}
    componentLinearAccScale=<1 1 1>
    componentLinearVelScale=<1 1 1>
    gravityOverride {z=-980}
    physicsBodyDefinitions [{
      boundBone {boneName=dyn_chain_fr_l}
      boxExtents=<20 20 20>
      constraintSetup {angularLimitsMax=<5 -5 5>, angularLimitsMin=<-5 -25 -5>, angularTarget=<1 0 0>}
      localJointOffset=<3 0 0>
     }]
    relativeSpaceBone {boneName=head}
    simSpaceSettings {maxAngularAcceleration=35, maxAngularVelocity=7, simSpaceAngularAlpha=0.4}
   }
   n10 = Anim Dynamics - Bone: dyn_chain_fr_r(ComponentPose=@n9, bAlphaBoolEnabled=True, ExternalForce="0,0,0") {
    alphaInputType=Bool
    angularDampingOverride=1
    angularSpringConstant=10000
    bAngularSpring=true
    bOverrideAngularDamping=true
    bUseGravityOverride=true
    gravityScale=0.35
    simulationSpace=Actor
    alphaBoolBlend {blendInTime=0.2, blendOption=Linear, blendOutTime=0.2}
    boundBone {boneName=dyn_chain_fr_r}
    componentLinearAccScale=<1 1 1>
    componentLinearVelScale=<1 1 1>
    gravityOverride {z=-980}
    physicsBodyDefinitions [{
      boundBone {boneName=dyn_chain_fr_r}
      boxExtents=<20 20 20>
      constraintSetup {angularLimitsMax=<5 -5 5>, angularLimitsMin=<-5 -25 -5>, angularTarget=<1 0 0>}
      localJointOffset=<3 0 0>
     }]
    relativeSpaceBone {boneName=head}
    simSpaceSettings {maxAngularAcceleration=35, maxAngularVelocity=7, simSpaceAngularAlpha=0.4}
   }
   n11 = Anim Dynamics - Bone: dyn_chain_bk(ComponentPose=@n10, bAlphaBoolEnabled=True, ExternalForce="0,0,0") {
    alphaInputType=Bool
    angularDampingOverride=1
    angularSpringConstant=10000
    bAngularSpring=true
    bOverrideAngularDamping=true
    bUseGravityOverride=true
    gravityScale=0.35
    simulationSpace=World
    alphaBoolBlend {blendInTime=0.2, blendOption=Linear, blendOutTime=0.2}
    boundBone {boneName=dyn_chain_bk}
    componentLinearAccScale=<1 1 1>
    componentLinearVelScale=<1 1 1>
    gravityOverride {z=-980}
    physicsBodyDefinitions [{
      boundBone {boneName=dyn_chain_bk}
      boxExtents=<15 15 15>
      constraintSetup {angularLimitsMax=<10 0 90>, angularLimitsMin=<-10 -90 -90>, angularTarget=<1 0 0>}
      localJointOffset=<3 0 0>
     }]
    simSpaceSettings {maxAngularAcceleration=35, maxAngularVelocity=7, simSpaceAngularAlpha=1}
   }
   n12 = Anim Dynamics - Bone: dyn_chain_bk_l(ComponentPose=@n11, bAlphaBoolEnabled=True, ExternalForce="0,0,0") {
    alphaInputType=Bool
    angularDampingOverride=1
    angularSpringConstant=10000
    bAngularSpring=true
    bOverrideAngularDamping=true
    bUseGravityOverride=true
    gravityScale=0.35
    simulationSpace=Actor
    alphaBoolBlend {blendInTime=0.2, blendOption=Linear, blendOutTime=0.2}
    boundBone {boneName=dyn_chain_bk_l}
    componentLinearAccScale=<1 1 1>
    componentLinearVelScale=<1 1 1>
    gravityOverride {z=-980}
    physicsBodyDefinitions [{
      boundBone {boneName=dyn_chain_bk_l}
      boxExtents=<20 20 20>
      constraintSetup {angularLimitsMax=<10 0 10>, angularLimitsMin=<-10 -25 -10>, angularTarget=<1 0 0>}
      localJointOffset=<3 0 0>
     }]
    relativeSpaceBone {boneName=head}
    simSpaceSettings {maxAngularAcceleration=35, maxAngularVelocity=7, simSpaceAngularAlpha=0.4}
   }
   n13 = Anim Dynamics - Bone: dyn_chain_bk_r(ComponentPose=@n12, bAlphaBoolEnabled=True, ExternalForce="0,0,0") {
    alphaInputType=Bool
    angularDampingOverride=1
    angularSpringConstant=10000
    bAngularSpring=true
    bOverrideAngularDamping=true
    bUseGravityOverride=true
    gravityScale=0.35
    simulationSpace=Actor
    alphaBoolBlend {blendInTime=0.2, blendOption=Linear, blendOutTime=0.2}
    boundBone {boneName=dyn_chain_bk_r}
    componentLinearAccScale=<1 1 1>
    componentLinearVelScale=<1 1 1>
    gravityOverride {z=-980}
    physicsBodyDefinitions [{
      boundBone {boneName=dyn_chain_bk_r}
      boxExtents=<20 20 20>
      constraintSetup {angularLimitsMax=<10 0 10>, angularLimitsMin=<-10 -25 -10>, angularTarget=<1 0 0>}
      localJointOffset=<3 0 0>
     }]
    relativeSpaceBone {boneName=head}
    simSpaceSettings {maxAngularAcceleration=35, maxAngularVelocity=7, simSpaceAngularAlpha=0.4}
   }
   n14 = Component To Local(ComponentPose=@n13)
   n15 = Get bIsInFrontEnd()
   n16 = "Constraint - dyn_chain_fr_l (P)"(
    ComponentPose=Local To Component(LocalPose=Blend Poses by bool(BlendPose_0=@n14, BlendPose_1=@n7, bActiveValue=@n15))
    Alpha=0.2
    ConstraintWeights_0=1
   ) {boneToModify {boneName=dyn_chain_fr_l}, constraintSetup [{targetBone {boneName=head}}], constraintWeights [1]}
   n17 = "Constraint - dyn_chain_fr_r (P)"(ComponentPose=@n16, Alpha=0.2, ConstraintWeights_0=1) {boneToModify {boneName=dyn_chain_fr_r}, constraintSetup [{targetBone {boneName=head}}], constraintWeights [1]}
   Output Pose(Result=Component To Local(ComponentPose=@n17)) {layerGroup=None}
   n18 = Layered blend per bone(
    BasePose=Copy Pose From Mesh() {bCopyCurves=true, bUseAttachedParent=true}
    BlendPoses_0=Copy Pose From Mesh(SourceMeshComponent=Get Body_Mesh())
   ) {layerSetup [{branchFilters [
       {boneName=FACIAL_C_FacialRoot}
       {boneName=FACIAL_C_Jaw}
       {boneName=FACIAL_C_Jawline}
       {boneName=FACIAL_L_ChinSide}
       {boneName=FACIAL_R_ChinSide}
       {boneName=FACIAL_C_Chin}
       {boneName=FACIAL_C_MouthLower}
       {boneName=FACIAL_C_LipLower}
       {boneName=FACIAL_L_LipLower}
       {boneName=FACIAL_R_LipLower}
       {boneName=FACIAL_L_LipLowerOuter}
       {boneName=FACIAL_R_LipLowerOuter}
       {boneName=FACIAL_L_Jawline}
       {boneName=FACIAL_R_Jawline}
       {boneName=FACIAL_L_CheekLower}
       {boneName=FACIAL_L_Masseter}
       {boneName=FACIAL_R_Masseter}
       {boneName=FACIAL_L_NasolabialBulge}
       {boneName=FACIAL_R_CheekLower}
       {boneName=FACIAL_C_MouthUpper}
       {boneName=FACIAL_C_LipUpper}
       {boneName=FACIAL_L_LipUpper}
       {boneName=FACIAL_R_LipUpper}
       {boneName=FACIAL_L_LipUpperOuter}
       {boneName=FACIAL_R_LipUpperOuter}
       {boneName=FACIAL_L_LipCorner}
       {boneName=FACIAL_R_LipCorner}
      ]}]}
   n19 = "Transform (Modify) Bone - Bone: base_chain_fr_l"(ComponentPose=Local To Component(LocalPose=@n18), Alpha=1, Translation="0,0,0.2", Rotation="0,0,0", Scale="1,1,1") {
    rotationMode=BMM_Additive
    rotationSpace=BCS_BoneSpace
    translationMode=BMM_Additive
    translationSpace=BCS_BoneSpace
    boneToModify {boneName=base_chain_fr_l}
    translation {z=0.2}
   }
   n20 = "Transform (Modify) Bone - Bone: base_chain_fr_r"(ComponentPose=@n19, Alpha=1, Translation="0,0,0.2", Rotation="0,0,0", Scale="1,1,1") {
    rotationMode=BMM_Additive
    rotationSpace=BCS_BoneSpace
    translationMode=BMM_Additive
    translationSpace=BCS_BoneSpace
    boneToModify {boneName=base_chain_fr_r}
    translation {z=0.2}
   }
   n21 = "Transform (Modify) Bone - Bone: FACIAL_C_Jaw"(ComponentPose=@n20, Alpha=1, Translation="0,0.5,1", Rotation="0,0,0", Scale="1,1,1.05") {
    rotationMode=BMM_Additive
    rotationSpace=BCS_BoneSpace
    scaleMode=BMM_Additive
    scaleSpace=BCS_BoneSpace
    translationMode=BMM_Additive
    translationSpace=BCS_BoneSpace
    boneToModify {boneName=FACIAL_C_Jaw}
    scale {z=1.05}
    translation {y=0.5, z=1}
   }
   Save cached pose 'falloff_cache'(Pose=Component To Local(ComponentPose=@n21)) {cacheName=falloff_cache}
  ]}
 EventGraph {nodes [
   Event OnCharacterCustomizationLoaded()
   > n1 = Cast To PlayerPawn_Athena(Object=TryGetPawnOwner())
   > Set Body_Mesh(Body_Mesh=GetSkeletalMeshForPartType(self=@n1, PartType=Head))
   Event BlueprintUpdateAnimation()
  ]}
]
variables [Body_Mesh {type=SkeletalMeshComponent}]
